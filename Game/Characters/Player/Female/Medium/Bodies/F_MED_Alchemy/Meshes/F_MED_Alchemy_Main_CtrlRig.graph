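# Control Rig for the female medium Alchemy body that procedurally corrects skirt, pouch and flap bones from the leg pose.
# A knee-aim bone is constrained to both calves and aimed at the right calf and pelvis, then steers the skirt split, which also follows both front skirt bases.
# Left-thigh rotation relative to the pelvis becomes a clamped weight for pouch and front-flap offsets.
# Rope and skirt-aim nodes sit outside the execution chain.
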ControlRigBlueprint /Game/Characters/Player/Female/Medium/Bodies/F_MED_Alchemy/Meshes/F_MED_Alchemy_Main_CtrlRig
graphs [
 RigVMFunctionLibrary {links [], nodes []}
 RigVMModel {nodes [
   RigUnit_BeginExecution = &RigUnit_BeginExecution()
   /Script/ControlRig.RigUnit_TransformConstraint(
    BaseTransform=(Rotation=(X=0,Y=0,Z=0,W=1),Translation=(X=0,Y=0,Z=0),Scale3D=(X=1,Y=1,Z=1))
    BaseTransformSpace=GlobalSpace
    Bone=dyn_kneeAim_l
    Targets="((Transform=(Rotation=(X=0,Y=0,Z=0,W=1),Translation=(X=0,Y=0,Z=0),Scale3D=(X=1,Y=1,Z=1)),Weight=1,bMaintainOffset=True,Filter=(TranslationFilter=(bX=True,bY=True,bZ=True),RotationFilter=(bX=True,bY=True,bZ=True),ScaleFilter=(bX=True,bY=True,bZ=True))),(Tra ...(truncated)"
    bUseInitialTransforms=True
    Targets.0.Transform.Translation=&RigUnit_GetBoneTransform(Bone=calf_l).Transform.Translation
    Targets.1.Transform.Translation=&RigUnit_GetBoneTransform(Bone=calf_r).Transform.Translation
    ExecutePin=@RigUnit_BeginExecution.ExecutePin
   ) {is=struct RigVMUnitNode}
   RigVMStructNode_0_1 = &RigUnit_GetBoneTransform(Bone=calf_r)
   > /Script/ControlRig.RigUnit_AimBone(
    Bone=dyn_kneeAim_l
    DebugSettings=(bEnabled=true,Scale=10,WorldOffset=(Rotation=(X=0,Y=0,Z=0,W=1),Translation=(X=0,Y=0,Z=0),Scale3D=(X=1,Y=1,Z=1)))
    Primary=(Weight=1,Axis=(X=0,Y=-0.1,Z=1),Target=(X=1,Y=0,Z=0),Kind=Location)
    Secondary=(Weight=1,Axis=(X=1,Y=-0.11,Z=-0.085),Target=(X=1,Y=0,Z=0),Kind=Location)
    Weight=1
    bPropagateToChildren=True
    Secondary.Target=&RigUnit_GetBoneTransform(Bone=pelvis).Transform.Translation
    Primary.Target=@RigVMStructNode_0_1.Transform.Translation
   ) {is=struct RigVMUnitNode}
   n1 = &RigUnit_GetBoneTransform(Bone=dyn_skirtSplitMain_01)
   > &RigUnit_TransformConstraint(
    BaseTransform=(Rotation=(X=0,Y=0,Z=0,W=1),Translation=(X=0,Y=0,Z=0),Scale3D=(X=1,Y=1,Z=1))
    BaseTransformSpace=GlobalSpace
    Bone=dyn_skirtSplitMain_01
    Targets="((Transform=(Rotation=(X=0,Y=0,Z=0,W=1),Translation=(X=0,Y=0,Z=0),Scale3D=(X=1,Y=1,Z=1)),Weight=1,bMaintainOffset=True,Filter=(TranslationFilter=(bX=True,bY=True,bZ=True),RotationFilter=(bX=True,bY=True,bZ=True),ScaleFilter=(bX=True,bY=True,bZ=True))),(Tra ...(truncated)"
    bUseInitialTransforms=True
    Targets.0.Transform.Rotation=&RigUnit_GetBoneTransform(Bone=dyn_kneeAim_l).Transform.Rotation
    Targets.1.Transform=@n1.Transform
    Targets.0.Transform.Translation=@n1.Transform.Translation
   )
   n2 = &RigUnit_GetBoneTransform(Bone=dyn_skirtSplitMain_01)
   > &RigUnit_TransformConstraint(
    BaseTransform=(Rotation=(X=0,Y=0,Z=0,W=1),Translation=(X=0,Y=0,Z=0),Scale3D=(X=1,Y=1,Z=1))
    BaseTransformSpace=GlobalSpace
    Bone=dyn_skirtSplitMain_01
    Targets="((Transform=(Rotation=(X=0,Y=0,Z=0,W=1),Translation=(X=0,Y=0,Z=0),Scale3D=(X=1,Y=1,Z=1)),Weight=1,bMaintainOffset=True,Filter=(TranslationFilter=(bX=True,bY=True,bZ=True),RotationFilter=(bX=True,bY=True,bZ=True),ScaleFilter=(bX=True,bY=True,bZ=True))),(Tra ...(truncated)"
    bUseInitialTransforms=True
    Targets.0.Transform.Translation=&RigUnit_GetBoneTransform(Bone=dyn_skirtA_01_Base_l).Transform.Translation
    Targets.1.Transform.Translation=&RigUnit_GetBoneTransform(Bone=dyn_skirtA_01_Base_r).Transform.Translation
    Targets.0.Transform.Rotation=@n2.Transform.Rotation
    Targets.1.Transform.Rotation=@n2.Transform.Rotation
   )
   n3 = &RigVMFunction_MathFloatRemap(
    SourceMaximum=-105
    SourceMinimum=-18
    TargetMinimum=0
    bClamp=true
    Value=&RigVMFunction_MathQuaternionToEuler(
     RotationOrder=ZYX
     Value=&RigUnit_GetRelativeTransformForItem(Child=(Type=Bone,Name="thigh_l"), Parent=(Type=Bone,Name="pelvis")).RelativeTransform.Rotation
    ).Result.Z
   )
   > &RigUnit_ModifyTransforms(
    ItemToModify=((Item=(Type=Bone,Name="dyn_pouchC_01_Base"),Transform=(Rotation=(X=0,Y=0,Z=0,W=1),Translation=(X=2,Y=-1,Z=-6),Scale3D=(X=1,Y=1,Z=1))))
    Weight=@n3.Result
    ItemToModify.0.Transform.Rotation=&RigVMFunction_MathQuaternionFromEuler(Euler=(X=20,Y=0,Z=-10)).Result
   )
   > &RigUnit_ModifyTransforms(
    ItemToModify=((Item=(Type=Bone,Name="dyn_pouchB_01_Base"),Transform=(Rotation=(X=0,Y=0,Z=0,W=1),Translation=(X=2,Y=-1,Z=-9),Scale3D=(X=1,Y=1,Z=1))))
    Weight=@n3.Result
    ItemToModify.0.Transform.Rotation=&RigVMFunction_MathQuaternionFromEuler(Euler=(X=35,Y=0,Z=-10)).Result
   )
   > &RigUnit_ModifyTransforms(
    ItemToModify=((Item=(Type=Bone,Name="dyn_pouchA_01_Base"),Transform=(Rotation=(X=0,Y=0,Z=0,W=1),Translation=(X=-4,Y=5,Z=-10),Scale3D=(X=1,Y=1,Z=1))))
    Weight=@n3.Result
    ItemToModify.0.Transform.Rotation=&RigVMFunction_MathQuaternionFromEuler(Euler=(X=40,Y=-30,Z=-90)).Result
   )
   > &RigUnit_ModifyTransforms(
    ItemToModify=((Item=(Type=Bone,Name="dyn_frontFlapCenter_01_Base"),Transform=(Rotation=(X=0,Y=0,Z=0,W=1),Translation=(X=-3,Y=-2,Z=-1),Scale3D=(X=1,Y=1,Z=1))))
    Weight=@n3.Result
    ItemToModify.0.Transform.Rotation=&RigVMFunction_MathQuaternionFromEuler(Euler=(X=-35,Y=30,Z=-100)).Result
   )
   > &RigUnit_ModifyTransforms(
    ItemToModify=((Item=(Type=Bone,Name="dyn_frontFlap_01_Base_r"),Transform=(Rotation=(X=0,Y=0,Z=0,W=1),Translation=(X=-2,Y=-1,Z=1),Scale3D=(X=1,Y=1,Z=1))))
    Weight=@n3.Result
    ItemToModify.0.Transform.Rotation=&RigVMFunction_MathQuaternionFromEuler(Euler=(X=0,Y=30,Z=0)).Result
   )
   n4 = &RigUnit_GetBoneTransform(Bone=dyn_skirtD_02_l)
   &RigUnit_TransformConstraint(BaseTransformSpace=GlobalSpace, Bone=dyn_innerRope_01, ExecutePin=(), Targets=(), bUseInitialTransforms=True, BaseTransform=@n4.Transform)
   > &RigUnit_TransformConstraint(BaseTransformSpace=GlobalSpace, Bone=dyn_outerRope_01, Targets=(), bUseInitialTransforms=True, BaseTransform=@n4.Transform)
   n5 = &RigUnit_GetBoneTransform(Bone=dyn_kneeAim_l)
   RigVMStructNode_0 = &RigUnit_GetBoneTransform(Bone=dyn_kneeAim_l)
   &RigUnit_AimBone(
    Bone=dyn_skirtAim
    DebugSettings=(Scale=10,WorldOffset=(Rotation=(X=0,Y=0,Z=0,W=1),Translation=(X=0,Y=0,Z=0),Scale3D=(X=1,Y=1,Z=1)))
    ExecutePin=()
    Primary=(Weight=1,Axis=(X=1,Y=0,Z=0),Target=(X=1,Y=0,Z=0),Kind=Location)
    Secondary=(Weight=1,Axis=(X=0,Y=-0.1,Z=0),Target=(X=1,Y=0,Z=0),Kind=Direction)
    Weight=1
    bPropagateToChildren=True
    Secondary.Target=@n5.Transform.Translation
    Primary.Target=@RigVMStructNode_0.Transform.Translation
   )
   &RigUnit_GetBoneTransform(Bone=dyn_skirtAim)
   &RigUnit_TransformConstraint(
    BaseTransform=(Rotation=(X=0,Y=0,Z=0,W=1),Translation=(X=0,Y=0,Z=0),Scale3D=(X=1,Y=1,Z=1))
    BaseTransformSpace=GlobalSpace
    Bone=dyn_skirtAim
    ExecutePin=()
    Targets=()
    bUseInitialTransforms=True
    BaseTransform.Translation=&RigUnit_GetBoneTransform(Bone=dyn_skirtSplitMain_01).Transform.Translation
    BaseTransform.Rotation=@n5.Transform.Rotation
   )
  ]}
]
hierarchy [
 root {loc=<0 0 0>, rot=<0 0 0 1>}
 pelvis {parent=root, loc=<0 -0.5301 98.0571>, rot=<0.0127 0.707 -0.0127 -0.707>}
 spine_01 {parent=pelvis, loc=<6.1208 -0.9328 0>, rot=<0 0 0.034 -0.9994>}
 spine_02 {parent=spine_01, loc=<6.5996 -0.2095 0>, rot=<0 0 0.0238 -0.9997>}
 spine_03 {parent=spine_02, loc=<6.4914 -0.1237 0>, rot=<0 0 0.0149 -0.9999>}
 spine_04 {parent=spine_03, loc=<6.9598 0.0794 0>, rot=<0 0 -0.0132 -0.9999>}
 spine_05 {parent=spine_04, loc=<6.9987 0 0>, rot=<0 0 -0.0015 -1>}
 clavicle_l {parent=spine_05, loc=<5.4831 -0.1053 -3.2946>, rot=<0.0465 -0.7289 -0.009 -0.6829>}
 upperarm_l {parent=clavicle_l, loc=<9.8658 -1.2136 -0.4105>, rot=<0.0164 -0.3245 -0.0269 -0.9454>}
 lowerarm_l {parent=upperarm_l, loc=<26.465 -0.8345 0.259>, rot=<0.0036 0.0161 -0.2298 -0.9731>}
 hand_l {parent=lowerarm_l, loc=<22.7355 -0.3203 -0.6105>, rot=<-0.6784 -0.0223 0.0735 -0.7307>}
 index_metacarpal_l {parent=hand_l, loc=<1.6535 0.0183 -2.96>, rot=<0.0158 -0.0124 -0.0173 -0.9996>}
 index_01_l {parent=index_metacarpal_l, loc=<8.3008 0.0357 -0.0524>, rot=<-0.0044 0.0485 0.0924 -0.9945>}
 index_02_l {parent=index_01_l, loc=<3.3587 -0.3003 0.0452>, rot=<-0.002 0.0027 0.3225 -0.9466>}
 index_03_l {parent=index_02_l, loc=<2.5578 -0.123 -0.0252>, rot=<-0.002 0.0067 0.2558 -0.9667>}
 middle_metacarpal_l {parent=hand_l, loc=<2.1856 -0.1725 -0.5984>, rot=<0.006 0.0168 -0.0329 -0.9993>}
 middle_01_l {parent=middle_metacarpal_l, loc=<7.3275 -0.2447 0.0129>, rot=<-0.0208 0.0042 0.0858 -0.9961>}
 middle_02_l {parent=middle_01_l, loc=<4.0123 0.1401 0.1175>, rot=<-0.0218 0.0258 0.3512 -0.9357>}
 middle_03_l {parent=middle_02_l, loc=<2.6077 0.0534 0.1373>, rot=<0.0082 0.0188 0.1708 -0.9851>}
 pinky_metacarpal_l {parent=hand_l, loc=<1.7616 -0.9753 3.3814>, rot=<-0.0397 0.0814 -0.069 -0.9935>}
 pinky_01_l {parent=pinky_metacarpal_l, loc=<7.1709 -0.4831 -0.1094>, rot=<-0.0654 -0.0035 0.2242 -0.9723>}
 pinky_02_l {parent=pinky_01_l, loc=<2.8274 0.0893 -0.1043>, rot=<0.0037 0.0011 0.2854 -0.9584>}
 pinky_03_l {parent=pinky_02_l, loc=<2.2046 0.0308 -0.0078>, rot=<-0.0002 0.0098 0.1445 -0.9895>}
 ring_metacarpal_l {parent=hand_l, loc=<2.2074 -0.3625 1.3638>, rot=<-0.0284 0.0722 -0.0458 -0.9959>}
 ring_01_l {parent=ring_metacarpal_l, loc=<7.0295 -0.6197 -0.1446>, rot=<-0.0489 -0.0252 0.1892 -0.9804>}
 ring_02_l {parent=ring_01_l, loc=<3.6174 0.4647 -0.1027>, rot=<0 0 0.2806 -0.9598>}
 ring_03_l {parent=ring_02_l, loc=<2.4152 0.1141 0.0199>, rot=<0 0 0.1189 -0.9929>}
 thumb_01_l {parent=hand_l, loc=<1.3118 -0.3191 -2.459>, rot=<0.6546 -0.4042 0.0664 -0.6354>}
 thumb_02_l {parent=thumb_01_l, loc=<5.9336 -0.0448 0.0401>, rot=<-0.0037 0.0213 0.298 -0.9543>}
 thumb_03_l {parent=thumb_02_l, loc=<2.2223 0.3682 -0.0907>, rot=<0.0118 -0.013 0.0066 -0.9998>}
 lowerarm_twist_01_l {parent=lowerarm_l, loc=<18.316 0 0>, rot=<0 0 0 1>}
 lowerarm_twist_02_l {parent=lowerarm_l, loc=<10.6707 0 0>, rot=<0 0 0 1>}
 dyn_forearmFlapA_01_Base_l {parent=lowerarm_twist_02_l, loc=<-8.4971 3.4 4.0906>, rot=<-0.3874 -0.661 0.6371 0.0847>}
 dyn_forearmFlapA_01_l {parent=dyn_forearmFlapA_01_Base_l, loc=<0 0 0>, rot=<0 0 0 1>}
 dyn_forearmFlapB_01_Base_l {parent=lowerarm_twist_02_l, loc=<-5.6295 -3.012 4.1652>, rot=<0.1559 -0.7521 0.5954 -0.2358>}
 dyn_forearmFlapB_01_l {parent=dyn_forearmFlapB_01_Base_l, loc=<0 0 0>, rot=<0 0 0 -1>}
 upperarm_twist_01_l {parent=upperarm_l, loc=<7.2135 -0.1134 0>, rot=<0 0 0 -1>}
 upperarm_twist_02_l {parent=upperarm_l, loc=<19.7998 -0.1134 0>, rot=<0 0 0 -1>}
 clavicle_r {parent=spine_05, loc=<5.4831 -0.1053 3.2946>, rot=<0.7289 0.0465 -0.6829 0.009>}
 upperarm_r {parent=clavicle_r, loc=<-9.8658 1.2136 0.4104>, rot=<0.0164 -0.3245 -0.0269 -0.9454>}
 lowerarm_r {parent=upperarm_r, loc=<-26.4651 0.8345 -0.259>, rot=<0.0036 0.0161 -0.2298 -0.9731>}
 hand_r {parent=lowerarm_r, loc=<-22.7356 0.3203 0.6106>, rot=<-0.6784 -0.0223 0.0735 -0.7307>}
 index_metacarpal_r {parent=hand_r, loc=<-1.6534 -0.0181 2.9601>, rot=<0.0158 -0.0124 -0.0173 -0.9996>}
 index_01_r {parent=index_metacarpal_r, loc=<-8.3008 -0.0357 0.0524>, rot=<-0.0044 0.0485 0.0924 -0.9945>}
 index_02_r {parent=index_01_r, loc=<-3.3589 0.3003 -0.0453>, rot=<-0.002 0.0027 0.3225 -0.9466>}
 index_03_r {parent=index_02_r, loc=<-2.5578 0.123 0.0252>, rot=<-0.002 0.0067 0.2558 -0.9667>}
 middle_metacarpal_r {parent=hand_r, loc=<-2.1856 0.1726 0.5984>, rot=<0.006 0.0168 -0.0329 -0.9993>}
 middle_01_r {parent=middle_metacarpal_r, loc=<-7.3273 0.2447 -0.0129>, rot=<-0.0208 0.0042 0.0858 -0.9961>}
 middle_02_r {parent=middle_01_r, loc=<-4.0124 -0.14 -0.1175>, rot=<-0.0218 0.0258 0.3512 -0.9357>}
 middle_03_r {parent=middle_02_r, loc=<-2.6077 -0.0533 -0.1373>, rot=<0.0082 0.0188 0.1708 -0.9851>}
 pinky_metacarpal_r {parent=hand_r, loc=<-1.7617 0.9753 -3.3814>, rot=<-0.0397 0.0814 -0.069 -0.9935>}
 pinky_01_r {parent=pinky_metacarpal_r, loc=<-7.1707 0.4832 0.1094>, rot=<-0.0654 -0.0035 0.2242 -0.9723>}
 pinky_02_r {parent=pinky_01_r, loc=<-2.8274 -0.0892 0.1043>, rot=<0.0037 0.0011 0.2854 -0.9584>}
 pinky_03_r {parent=pinky_02_r, loc=<-2.2045 -0.031 0.0078>, rot=<-0.0002 0.0098 0.1445 -0.9895>}
 ring_metacarpal_r {parent=hand_r, loc=<-2.2073 0.3627 -1.3638>, rot=<-0.0284 0.0722 -0.0458 -0.9959>}
 ring_01_r {parent=ring_metacarpal_r, loc=<-7.0297 0.6197 0.1446>, rot=<-0.0489 -0.0252 0.1892 -0.9804>}
 ring_02_r {parent=ring_01_r, loc=<-3.6172 -0.4647 0.1027>, rot=<0 0 0.2806 -0.9598>}
 ring_03_r {parent=ring_02_r, loc=<-2.4152 -0.1141 -0.0199>, rot=<0 0 0.1189 -0.9929>}
 thumb_01_r {parent=hand_r, loc=<-1.3119 0.3192 2.459>, rot=<0.6546 -0.4042 0.0664 -0.6354>}
 thumb_02_r {parent=thumb_01_r, loc=<-5.9336 0.0448 -0.0401>, rot=<-0.0037 0.0213 0.298 -0.9543>}
 thumb_03_r {parent=thumb_02_r, loc=<-2.2223 -0.3682 0.0907>, rot=<0.0118 -0.013 0.0066 -0.9998>}
 lowerarm_twist_01_r {parent=lowerarm_r, loc=<-18.3161 0 0>, rot=<0 0 0 1>}
 lowerarm_twist_02_r {parent=lowerarm_r, loc=<-10.6707 0 0>, rot=<0 0 0 1>}
 dyn_forearmFlap_01_Base_r {parent=lowerarm_twist_02_r, loc=<6.6587 -0.3473 -5.4214>, rot=<-0.6502 0.0707 0.1244 -0.7462>}
 dyn_forearmFlap_01_r {parent=dyn_forearmFlap_01_Base_r, loc=<0 0 0>, rot=<0 0 0 1>}
 dyn_wristDangleMain__Base_r {parent=lowerarm_twist_02_r, loc=<3.7627 -3.0414 1.9451>, rot=<0.0033 -0.8436 0.2545 0.4728>}
 dyn_wristDangleMain__r {parent=dyn_wristDangleMain__Base_r, loc=<0 0 0>, rot=<0 0 0 -1>}
 dyn_wristDangleB_01_r {parent=dyn_wristDangleMain__r, loc=<2.0959 0.2885 0.1959>, rot=<0.496 0.0608 -0.0444 -0.8651>}
 dyn_wristDangleB_02_r {parent=dyn_wristDangleB_01_r, loc=<2.7063 0 0>, rot=<-0.2386 -0.0679 0.0055 -0.9687>}
 dyn_wristDangleA_01_r {parent=dyn_wristDangleMain__r, loc=<2.1039 -0.0995 -0.1756>, rot=<0.0671 -0.0368 -0.0251 -0.9967>}
 dyn_wristDangleA_02_r {parent=dyn_wristDangleA_01_r, loc=<5.1702 0 0>, rot=<0.0992 -0.0307 0.0123 -0.9945>}
 upperarm_twist_01_r {parent=upperarm_r, loc=<-7.2135 0.1134 0>, rot=<0 0 0 1>}
 upperarm_twist_02_r {parent=upperarm_r, loc=<-19.7998 0.1134 0>, rot=<0 0 0 1>}
 neck_01 {parent=spine_05, loc=<11.3138 0.1529 0>, rot=<0 0 -0.2003 -0.9797>}
 neck_02 {parent=neck_01, loc=<4.5627 0 0>, rot=<0 0 0.0504 -0.9987>}
 head {parent=neck_02, loc=<5.3914 0 0>, rot=<0 0 0.115 -0.9934>}
 thigh_l {parent=pelvis, loc=<-1.0855 -0.2109 -8.7053>, rot=<0.0009 0.0355 0.0155 -0.9993>}
 calf_l {parent=thigh_l, loc=<-42.5284 -0.0019 -0.7297>, rot=<0.0006 -0.0161 -0.0391 -0.9991>}
 calf_twist_01_l {parent=calf_l, loc=<-35.8483 -1.9197 -0.5794>, rot=<0.0237 0.0095 -0.0035 -0.9997>}
 calf_twist_02_l {parent=calf_l, loc=<-19.1546 -1.7943 -0.2638>, rot=<0.0237 0.0095 -0.0035 -0.9997>}
 foot_l {parent=calf_l, loc=<-46.7401 -2.5014 -0.592>, rot=<0.0427 -0.0238 0.0353 -0.9982>}
 ball_l {parent=foot_l, loc=<-3.5658 10.8066 -0.1544>, rot=<0.0001 -0.0001 -0.7063 -0.7079>}
 dyn_legFlapMain_Base_l {parent=calf_l, loc=<-26.2665 -4.7934 -1.1208>, rot=<-0.0666 -0.9978 -0.0016 0.0024>}
 dyn_legFlapMain {parent=dyn_legFlapMain_Base_l, loc=<0 0 0>, rot=<0 0 0 -1>}
 dyn_legFlapBack_l {parent=dyn_legFlapMain, loc=<4.8647 -7.2244 2.0774>, rot=<-0.0011 -0.0154 -0.0218 -0.9996>}
 dyn_legFlapSide_l {parent=dyn_legFlapMain, loc=<4.9625 -0.061 -5.6142>, rot=<-0.7068 0.0099 0.0382 -0.7063>}
 dyn_Base_l {parent=calf_l, loc=<-8.0904 -4.0883 5.9737>, rot=<0.0831 -0.9855 -0.0261 0.1456>}
 dyn_legDangleMain_l {parent=dyn_Base_l, loc=<0 0 0>, rot=<0 0 0 1>}
 dyn_legDangleA_l {parent=dyn_legDangleMain_l, loc=<8.457 0.4746 -0.2851>, rot=<-0.4564 -0.0048 -0.042 -0.8888>}
 dyn_legDangleB_l {parent=dyn_legDangleMain_l, loc=<8.7645 0.9251 0.6839>, rot=<-0.5064 0.0118 -0.0645 -0.8598>}
 dyn_kneeAim_l {parent=calf_l, loc=<0 0 0>, rot=<0 0 0 1>}
 thigh_twist_01_l {parent=thigh_l, loc=<-27.952 0 0>, rot=<0.0013 0 0 -1>}
 dyn_skirtA_01_Base_l {parent=thigh_l, loc=<-8.6601 7.9367 2.8574>, rot=<0.0024 0.0581 0.9983 -0.0075>}
 dyn_skirtA_01_l {parent=dyn_skirtA_01_Base_l, loc=<0 0 0>, rot=<0 0 0 1>}
 dyn_skirtA_02_l {parent=dyn_skirtA_01_l, loc=<12.9038 0 0>, rot=<-0.0245 -0.0165 -0.0199 -0.9994>}
 dyn_skirtA_03_l {parent=dyn_skirtA_02_l, loc=<9.3291 0 0>, rot=<0.0252 -0.0147 0.0443 -0.9986>}
 dyn_skirtB_01_Base_l {parent=thigh_l, loc=<-8.4904 5.5216 -3.6817>, rot=<-0.0335 0.2204 0.9748 0.0095>}
 dyn_skirtB_01_l {parent=dyn_skirtB_01_Base_l, loc=<0 0 0>, rot=<0 0 0 1>}
 dyn_skirtB_02_l {parent=dyn_skirtB_01_l, loc=<12.5085 0 0>, rot=<0.0713 0.0009 -0.0437 -0.9965>}
 dyn_skirtB_03_l {parent=dyn_skirtB_02_l, loc=<8.8978 0 0>, rot=<0 0 0 -1>}
 dyn_skirtC_01_Base_l {parent=thigh_l, loc=<-7.879 -0.8405 -7.4804>, rot=<-0.069 0.6361 0.767 0.0485>}
 dyn_skirtC_01_l {parent=dyn_skirtC_01_Base_l, loc=<0 0 0>, rot=<0 0 0 -1>}
 dyn_skirtC_02_l {parent=dyn_skirtC_01_l, loc=<12.4735 0 0>, rot=<0.0259 -0.0366 -0.0536 -0.9976>}
 thigh_r {parent=pelvis, loc=<-1.0853 -0.2109 8.7053>, rot=<0.0355 -0.0009 0.9993 0.0155>}
 calf_r {parent=thigh_r, loc=<42.5286 0.0019 0.7297>, rot=<0.0006 -0.0161 -0.0391 -0.9991>}
 calf_twist_01_r {parent=calf_r, loc=<35.8485 1.9197 0.5794>, rot=<0.0237 0.0095 -0.0035 -0.9997>}
 calf_twist_02_r {parent=calf_r, loc=<19.1546 1.7943 0.2638>, rot=<0.0237 0.0095 -0.0035 -0.9997>}
 foot_r {parent=calf_r, loc=<46.7403 2.5014 0.592>, rot=<0.0427 -0.0238 0.0353 -0.9982>}
 ball_r {parent=foot_r, loc=<3.5658 -10.8066 0.1544>, rot=<0.0001 -0.0001 -0.7063 -0.7079>}
 dyn_legFlapMain_Base_r {parent=calf_r, loc=<25.8514 5.5084 1.3581>, rot=<0.9963 -0.0862 -0.0045 -0.0015>}
 dyn_legFlapMain_r {parent=dyn_legFlapMain_Base_r, loc=<0 0 0>, rot=<0 0 0 -1>}
 dyn_legFlapInner_r {parent=dyn_legFlapMain_r, loc=<5.5893 1.291 6.0576>, rot=<0.706 -0.0295 0.0583 -0.7052>}
 dyn_legFlapBack_r {parent=dyn_legFlapMain_r, loc=<5.0511 -6.7719 -0.1918>, rot=<0.0015 0.0165 -0.0021 -0.9999>}
 dyn_legFlapOuter_r {parent=dyn_legFlapMain_r, loc=<5.8568 0.3438 -5.6878>, rot=<-0.7057 0.0464 0.0295 -0.7063>}
 dyn_legDangleMainA_r {parent=calf_r, loc=<9.1104 8.1169 4.6561>, rot=<0.9259 0.1157 0.142 -0.3305>}
 dyn_legDangleBackA_01_r {parent=dyn_legDangleMainA_r, loc=<0.2294 -0.0353 0.9912>, rot=<0.0523 0.096 0.0259 -0.9937>}
 dyn_legDangleBackA_02_r {parent=dyn_legDangleBackA_01_r, loc=<4.3522 0 0>, rot=<-0.1462 -0.0197 -0.0566 -0.9874>}
 dyn_legDangleFrontA_01_r {parent=dyn_legDangleMainA_r, loc=<0.0987 0.3085 -0.8289>, rot=<-0.2731 -0.0084 0.0102 -0.9619>}
 dyn_legDangleFrontA_02_r {parent=dyn_legDangleFrontA_01_r, loc=<4.6656 0 0>, rot=<-0.0204 -0.0045 -0.0174 -0.9996>}
 dyn_legDangleMainB_01_r {parent=calf_r, loc=<17.9058 3.3033 7.982>, rot=<0.6487 -0.2505 0.2822 -0.6609>}
 dyn_legDangleMainB_02_r {parent=dyn_legDangleMainB_01_r, loc=<2.2743 0 0>, rot=<-0.0338 0.3737 0.1077 -0.9207>}
 dyn_legDangleBackB_01_r {parent=dyn_legDangleMainB_02_r, loc=<1.4234 0.0109 0.4748>, rot=<-0.1715 0.0859 0.0048 -0.9814>}
 dyn_legDangleBackB_02_r {parent=dyn_legDangleBackB_01_r, loc=<2.8574 0 0>, rot=<-0.0326 0.0417 -0.0168 -0.9985>}
 dyn_legDangleBackC_01_r {parent=dyn_legDangleMainB_02_r, loc=<2.8295 0.3731 -0.9078>, rot=<-0.276 -0.0233 -0.0082 -0.9609>}
 dyn_legDangleBackC_02_r {parent=dyn_legDangleBackC_01_r, loc=<2.8669 0 0>, rot=<0.0614 0.0287 0.0037 -0.9977>}
 thigh_twist_01_r {parent=thigh_r, loc=<27.952 0 0>, rot=<0.0013 0 0 -1>}
 dyn_skirtA_01_Base_r {parent=thigh_r, loc=<7.4553 -7.8903 -1.0987>, rot=<0.0669 0.0253 0.0177 -0.9973>}
 dyn_skirtA_01_r {parent=dyn_skirtA_01_Base_r, loc=<0 0 0>, rot=<0 0 0 1>}
 dyn_skirtA_02_r {parent=dyn_skirtA_01_r, loc=<13.6686 0 0>, rot=<-0.0226 0.0161 -0.0464 -0.9985>}
 dyn_skirtA_03_r {parent=dyn_skirtA_02_r, loc=<8.891 0 0>, rot=<0 0 -0.0349 -0.9994>}
 dyn_skirtB_01_Base_r {parent=thigh_r, loc=<5.9865 -2.9265 6.7874>, rot=<0.5705 0.1089 0.0059 -0.814>}
 dyn_skirtB_01_r {parent=dyn_skirtB_01_Base_r, loc=<0 0 0>, rot=<0 0 0 -1>}
 dyn_skirtB_02_r {parent=dyn_skirtB_01_r, loc=<13.9702 0 0>, rot=<-0.0399 -0.009 0.0074 -0.9991>}
 dyn_skirtB_03_r {parent=dyn_skirtB_02_r, loc=<9.2535 0 0>, rot=<0 0 -0.0349 -0.9994>}
 dyn_skirtBack_01_Base {parent=pelvis, loc=<-5.9731 -14.5801 -1.952>, rot=<0.0361 -0.9986 -0.0238 0.0291>}
 dyn_skirtBack_01 {parent=dyn_skirtBack_01_Base, loc=<0 0 0>, rot=<0 0 0 1>}
 dyn_skirtBack_02 {parent=dyn_skirtBack_01, loc=<10.0658 0 0>, rot=<0.0046 0.0145 -0.0243 -0.9996>}
 dyn_skirtBack_03 {parent=dyn_skirtBack_02, loc=<10.9115 0 0>, rot=<-0.0051 0.0059 -0.0416 -0.9991>}
 dyn_skirtBack_04 {parent=dyn_skirtBack_03, loc=<10.1005 0 0>, rot=<0.0109 0.0028 -0.0002 -0.9999>}
 dyn_skirtD_01_Base_r {parent=pelvis, loc=<-6.1142 -11.8528 8.8292>, rot=<0.0744 -0.9772 0.1949 0.0394>}
 dyn_skirtD_01_r {parent=dyn_skirtD_01_Base_r, loc=<0 0 0>, rot=<0 0 0 -1>}
 dyn_skirtD_02_r {parent=dyn_skirtD_01_r, loc=<9.6949 0 0>, rot=<-0.0033 0.0226 -0.0304 -0.9993>}
 dyn_skirtD_03_r {parent=dyn_skirtD_02_r, loc=<10.0405 0 0>, rot=<0.0434 0.0461 -0.0469 -0.9969>}
 dyn_skirtD_04_r {parent=dyn_skirtD_03_r, loc=<10.748 0 0>, rot=<0 0 0 -1>}
 dyn_skirtD_01_Base_l {parent=pelvis, loc=<-3.746 -9.6332 -13.0461>, rot=<0.0527 -0.9501 -0.3061 -0.028>}
 dyn_skirtD_01_l {parent=dyn_skirtD_01_Base_l, loc=<0 0 0>, rot=<0 0 0 -1>}
 dyn_skirtD_02_l {parent=dyn_skirtD_01_l, loc=<11.6493 0 0>, rot=<-0.0749 -0.0058 -0.0691 -0.9948>}
 dyn_skirtD_03_l {parent=dyn_skirtD_02_l, loc=<10.4653 0 0>, rot=<-0.0287 -0.0247 0.0067 -0.9993>}
 dyn_skirtD_04_l {parent=dyn_skirtD_03_l, loc=<10.005 0 0>, rot=<0.0232 -0.0055 -0.0101 -0.9997>}
 dyn_innerRope_01 {parent=dyn_skirtD_01_l, loc=<7.9305 -2.1817 0.9102>, rot=<0.0447 0.0068 -0.043 -0.9981>}
 dyn_innerRope_02 {parent=dyn_innerRope_01, loc=<7.877 0 0>, rot=<-0.0229 -0.0091 -0.0126 -0.9996>}
 dyn_innerRope_03 {parent=dyn_innerRope_02, loc=<8.1601 0 0>, rot=<0 0 0 -1>}
 dyn_outerRope_02 {parent=dyn_innerRope_02, loc=<-0.0011 -0.8804 0.4487>, rot=<0.0852 0.066 0.0978 -0.9894>}
 dyn_outerRope_03 {parent=dyn_outerRope_02, loc=<10.1796 0 0>, rot=<0 0 0 -1>}
 dyn_outerRope_01 {parent=dyn_skirtD_01_l, loc=<7.9305 -2.1817 0.9102>, rot=<0.0577 0.0347 0.0129 -0.9976>}
 dyn_frontFlapCenter_01_Base {parent=pelvis, loc=<5.2441 9.3193 -2.2021>, rot=<0.0074 -0.0346 0.9692 0.2436>}
 dyn_frontFlapCenter_01 {parent=dyn_frontFlapCenter_01_Base, loc=<0 0 0>, rot=<0 0 0 -1>}
 dyn_frontFlapCenter_02 {parent=dyn_frontFlapCenter_01, loc=<2.308 0 0>, rot=<0.0648 -0.0541 -0.1377 -0.9869>}
 dyn_frontFlap_01_Base_r {parent=pelvis, loc=<6.5393 8.0193 5.7599>, rot=<0.1527 -0.1808 0.9308 0.2786>}
 dyn_frontFlap_01_r {parent=dyn_frontFlap_01_Base_r, loc=<0 0 0>, rot=<0 0 0 1>}
 dyn_frontFlap_02_r {parent=dyn_frontFlap_01_r, loc=<2.2313 0 0>, rot=<0.0105 0.0232 -0.1271 -0.9916>}
 dyn_backFlap_01_Base_l {parent=pelvis, loc=<9.8379 -5.4216 -9.2927>, rot=<0.4075 -0.828 -0.3393 -0.1823>}
 dyn_backFlap_01_l {parent=dyn_backFlap_01_Base_l, loc=<0 0 0>, rot=<0 0 0 1>}
 dyn_backFlap_02_l {parent=dyn_backFlap_01_l, loc=<2.5722 0 0>, rot=<-0.0918 -0.1094 -0.0766 -0.9868>}
 dyn_backFlapCenter_01_Base {parent=pelvis, loc=<8.5485 -9.127 -2.5973>, rot=<0.4255 -0.8976 -0.102 -0.0529>}
 dyn_backFlapCenter_01 {parent=dyn_backFlapCenter_01_Base, loc=<0 0 0>, rot=<0 0 0 1>}
 dyn_backFlapCenter_02 {parent=dyn_backFlapCenter_01, loc=<4.9204 0 0>, rot=<-0.0109 -0.0098 -0.187 -0.9822>}
 dyn_hipFlap_Base_r {parent=pelvis, loc=<9.6789 -1.6291 11.3503>, rot=<0.4472 -0.6276 0.5614 0.3016>}
 dyn_hipFlap_r {parent=dyn_hipFlap_Base_r, loc=<0 0 0>, rot=<0 0 0 -1>}
 dyn_skirtSplitMain_01 {parent=pelvis, loc=<-9.8738 7.9883 1.1003>, rot=<-0.0002 -0.0104 0.9998 0.0179>}
 dyn_skirtSplit_01_l {parent=dyn_skirtSplitMain_01, loc=<0 -0.5312 0>, rot=<-0.0119 -0.1304 -0.0016 -0.9914>}
 dyn_skirtSplit_02_l {parent=dyn_skirtSplit_01_l, loc=<9.0591 0 0>, rot=<-0.0135 0.0617 -0.0358 -0.9974>}
 dyn_skirtSplit_03_l {parent=dyn_skirtSplit_02_l, loc=<8.6732 0 0>, rot=<-0.0098 -0.0014 0.0376 -0.9992>}
 dyn_skirtSplit_04_l {parent=dyn_skirtSplit_03_l, loc=<11.5739 0 0>, rot=<0.9949 -0.0592 -0.0751 0.0317>}
 dyn_skirtSplit_05_l {parent=dyn_skirtSplit_04_l, loc=<3.6866 0 0>, rot=<0.0355 0.5938 0.048 -0.8024>}
 dyn_skirtSplit_01_r {parent=dyn_skirtSplitMain_01, loc=<0 0.4937 0.0215>, rot=<-0.0972 0.0883 0.0423 -0.9904>}
 dyn_skirtSplit_02_r {parent=dyn_skirtSplit_01_r, loc=<8.6996 0 0>, rot=<-0.0002 -0.0156 -0.0545 -0.9984>}
 dyn_skirtSplit_03_r {parent=dyn_skirtSplit_02_r, loc=<8.9038 0 0>, rot=<0.0595 0.0265 -0.0294 -0.9974>}
 dyn_skirtSplit_04_r {parent=dyn_skirtSplit_03_r, loc=<9.8056 0 0>, rot=<-0.6265 0.0409 0.7783 0.0106>}
 dyn_skirtAim {parent=pelvis, loc=<-1.0854 -0.2109 0>, rot=<0 0 0.9999 0.0155>}
 dyn_hipSash_01 {parent=pelvis, loc=<-7.6287 -3.1889 -18.3691>, rot=<0.1131 -0.7136 -0.6786 -0.1327>}
 dyn_hipSash_02 {parent=dyn_hipSash_01, loc=<3.5865 0 0>, rot=<-0.0156 -0.0052 -0.028 -0.9995>}
 dyn_hipSash_03 {parent=dyn_hipSash_02, loc=<9.7195 0 0>, rot=<-0.029 -0.0202 -0.0558 -0.9978>}
 dyn_skirtSashBottomBase {parent=dyn_hipSash_03, loc=<3.0564 0.8326 3.1937>, rot=<0.0197 -0.0745 0.0398 -0.9962>}
 dyn_skirtSashBottom_01 {parent=dyn_skirtSashBottomBase, loc=<6.5411 0 0>, rot=<-0.0381 -0.0783 -0.0573 -0.9945>}
 dyn_skirtSashBottom_02 {parent=dyn_skirtSashBottom_01, loc=<6.5111 0 0>, rot=<0.0228 -0.1109 -0.0204 -0.9934>}
 dyn_skirtSashTop_01 {parent=dyn_hipSash_03, loc=<6.5459 -0.6566 -1.7241>, rot=<-0.0213 0.0261 0.0045 -0.9994>}
 dyn_skirtSashTop_02 {parent=dyn_skirtSashTop_01, loc=<9.6017 0 0>, rot=<-0.0287 -0.0522 -0.0317 -0.9977>}
 dyn_pouchA_01_Base {parent=pelvis, loc=<1.5162 9.2244 -6.9053>, rot=<-0.0802 0.0958 0.9916 0.033>}
 dyn_pouchA_01 {parent=dyn_pouchA_01_Base, loc=<0 0 0>, rot=<0 0 0 -1>}
 dyn_pouchA_02 {parent=dyn_pouchA_01, loc=<6.0504 0 0>, rot=<0.0208 -0.0018 -0.0314 -0.9993>}
 dyn_pouchB_01_Base {parent=pelvis, loc=<4.273 7.5929 -11.0409>, rot=<-0.1584 0.384 0.9061 0.0804>}
 dyn_pouchB_01 {parent=dyn_pouchB_01_Base, loc=<0 0 0>, rot=<0 0 0 1>}
 dyn_pouchB_02 {parent=dyn_pouchB_01, loc=<4.95 0 0>, rot=<0.0083 -0.0147 -0.1092 -0.9939>}
 dyn_pouchC_01_Base {parent=pelvis, loc=<5.3643 -0.8435 -13.6973>, rot=<-0.2186 0.6456 0.7167 0.1472>}
 dyn_pouchC_01 {parent=dyn_pouchC_01_Base, loc=<0 0 0>, rot=<0 0 0 1>}
 dyn_pouchC_02 {parent=dyn_pouchC_01, loc=<7.3447 0 0>, rot=<0.0095 0.046 0.0115 -0.9988>}
 dyn_quiverMain_Base {parent=pelvis, loc=<2.3673 -4.337 14.1888>, rot=<0.4533 -0.8429 -0.2712 0.1023>}
 dyn_quiverMain {parent=dyn_quiverMain_Base, loc=<0 0 0>, rot=<0 0 0 1>}
 dyn_arrowA {parent=dyn_quiverMain, loc=<14.9386 0.5231 -2.2148>, rot=<-0.0929 -0.9931 0.0702 -0.0109>}
 dyn_arrowB {parent=dyn_quiverMain, loc=<21.9287 -1.6753 -1.0995>, rot=<-0.1121 -0.9885 -0.1011 -0.0088>}
 dyn_arrowC {parent=dyn_quiverMain, loc=<19.2654 1.5706 -0.3231>, rot=<-0.1163 -0.9914 0.0434 -0.0405>}
 dyn_arrowD {parent=dyn_quiverMain, loc=<21.9287 -1.6753 -1.0995>, rot=<-0.1765 -0.9842 -0.0085 -0.0148>}
]
