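# Physics asset for the female medium Alchemy body's secondary motion, apparently the front-end variant judging by its name.
# Capsules cover sashes, ropes, flaps, pouches, quiver, shirt and dangles; boxes form skirt panels; a few spine and leg bodies represent the character.
# Dynamic chains use tight swing and twist limits from Base or root bodies.
# Some bases link to spine_05, thighs or calf_l with free angular constraints; others have no parent constraint.
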PhysicsAsset /Game/Characters/Player/Female/Medium/Bodies/F_MED_Alchemy/Meshes/F_MED_Alchemy_PerformancePhysicsFrontEnd
bodies {
 calf_l [{shapeType=Capsule, length=0, radius=0.5}]
 calf_r [{shapeType=Capsule, length=0, radius=0.5}]
 dyn_Base_l [{shapeType=Capsule, length=15.7313, radius=6.655, center=<6.9586 1.8674 8.1008>, rotation=<rot 4.5496 90.7693 -81.4674>}]
 dyn_backFlapCenter_01 [{shapeType=Capsule, length=4.3161, radius=1.7375, center=<2.2031 0 0>, rotation=<rot 90 0 0>}]
 dyn_backFlapCenter_01_Base [{shapeType=Capsule, length=0.505, radius=0.505, rotation=<rot 0 -90 0>}]
 dyn_backFlapCenter_02 [{shapeType=Capsule, length=6.4574, radius=2.6199, center=<3.2163 0 0>, rotation=<rot 90 0 0>}]
 dyn_backFlap_01_Base_l [{shapeType=Capsule, length=0.505, radius=0.505, rotation=<rot 0 -90 0>}]
 dyn_backFlap_01_l [{shapeType=Capsule, length=1.9272, radius=1.7375, center=<1.2013 0 0>, rotation=<rot 90 0 0>}]
 dyn_backFlap_02_l [{shapeType=Capsule, length=2.9329, radius=1.4212, center=<1.634 0 0>, rotation=<rot 90 0 0>}]
 dyn_forearmFlapA_01_Base_l [{shapeType=Capsule, length=0.505, radius=0.505, rotation=<rot 0 -90 0>}]
 dyn_forearmFlapA_01_l [{shapeType=Capsule, length=3.6586, radius=1.6074, center=<1.0783 0 0>, rotation=<rot 90 0 0>}]
 dyn_forearmFlapB_01_Base_l [{shapeType=Capsule, length=0.505, radius=0.505, rotation=<rot 0 -90 0>}]
 dyn_forearmFlapB_01_l [{shapeType=Capsule, length=4.6189, radius=2.1243, center=<1.5428 0 -0.0002>, rotation=<rot 90 0 0>}]
 dyn_forearmFlap_01_Base_r [{shapeType=Capsule, length=0.505, radius=0.505, rotation=<rot 0 -90 0>}]
 dyn_forearmFlap_01_r [{shapeType=Capsule, length=3.6586, radius=1.6074, center=<1.0783 0 0>, rotation=<rot 90 0 0>}]
 dyn_hipSash_01 [{shapeType=Capsule, length=0.505, radius=0.505, rotation=<rot 0 -90 0>}]
 dyn_hipSash_02 [{shapeType=Capsule, length=6.7613, radius=5.2091, center=<4.6968 -3.25 0>, rotation=<rot 0 -89.9339 -87.2772>}]
 dyn_hipSash_03 [{shapeType=Capsule, length=2.0092, radius=5.2091, center=<4.6968 -3.25 0>, rotation=<rot 0 -89.9339 -87.2772>}]
 dyn_innerRope_01 [{shapeType=Capsule, length=4.5277, radius=2.3682, center=<3.5 0 0>, rotation=<rot 90 0 0>}]
 dyn_innerRope_02 [{shapeType=Capsule, length=5.6292, radius=4.1427, center=<3.5 -3 0>, rotation=<rot 90 0 0>}]
 dyn_innerRope_03 [{shapeType=Capsule, length=3.1436, radius=5.2906, center=<3.5 -2 0>, rotation=<rot 90 0 0>}]
 dyn_legDangleA_l [{shapeType=Capsule, length=5.7522, radius=0.8984, center=<3.3648 0.0091 0.0151>, rotation=<rot 0 -90.1614 -89.097>}]
 dyn_legDangleB_l [{shapeType=Capsule, length=6.3451, radius=0.92, center=<3.9358 0.0352 -0.0156>, rotation=<rot 0 -90.5814 -88.5687>}]
 dyn_legDangleMain_l [{shapeType=Capsule, length=8.4047, radius=0.8904, center=<3.9459 0.0002 -0.2226>, rotation=<rot 85.3987 -179.9999 180>}]
 dyn_outerRope_02 [{shapeType=Capsule, length=5.6292, radius=4.1427, center=<4.7572 -3 0>, rotation=<rot 90 0.1583 0.1583>}]
 dyn_outerRope_03 [{shapeType=Capsule, length=4.8936, radius=3.5406, center=<3.5 -2 0>, rotation=<rot 90 0 0>}]
 dyn_pouchA_01 [{shapeType=Capsule, length=3.4543, radius=1.9156, center=<3 -1 0>, rotation=<rot 90 0 0>}]
 dyn_pouchA_01_Base [{shapeType=Capsule, length=1.9132, radius=1.0066}]
 dyn_pouchB_01 [{shapeType=Capsule, length=4.6763, radius=1.9156, center=<2.6094 -1 0>, rotation=<rot 90 0 0>}]
 dyn_pouchB_01_Base [{shapeType=Capsule, length=1.9132, radius=1.0066}]
 dyn_pouchC_01 [{shapeType=Capsule, length=6.2086, radius=1.9156, center=<3.5288 -1 0>, rotation=<rot 90 0 0>}]
 dyn_pouchC_01_Base [{shapeType=Capsule, length=1.9132, radius=1.0066}]
 dyn_quiverMain [{shapeType=Capsule, length=40.4002, radius=7.7049, center=<11.3359 1.6064 -4.281>, rotation=<rot 0 77.9998 89.9999>}]
 dyn_quiverMain_Base [{shapeType=Capsule, length=0.505, radius=0.505, rotation=<rot 0 -90 0>}]
 dyn_shirtArmpit_01_Base_l [{shapeType=Capsule, length=0.505, radius=0.505, rotation=<rot 0 -90 0>}]
 dyn_shirtArmpit_01_Base_r [{shapeType=Capsule, length=0.505, radius=0.505, rotation=<rot 0 -90 0>}]
 dyn_shirtArmpit_01_l [{shapeType=Capsule, length=7.8, radius=2.046, center=<4.5 0 0>, rotation=<rot 90 0 0>}]
 dyn_shirtArmpit_01_r [{shapeType=Capsule, length=7.8, radius=2.046, center=<4.5 0 0>, rotation=<rot 90 0 180>}]
 dyn_shirtArmpit_02_l [{shapeType=Capsule, length=2.7277, radius=1.6677, center=<1.6 -0.9942 0>, rotation=<rot 84.5763 0 0>}]
 dyn_shirtArmpit_02_r [{shapeType=Capsule, length=2.7277, radius=1.6677, center=<1.6 -1 0>, rotation=<rot 84.5763 0 180>}]
 dyn_shirtArmpit_03_l [{shapeType=Capsule, length=2.6153, radius=2.0364, center=<2.0743 -1.5 0>, rotation=<rot 90 0 0>}]
 dyn_shirtArmpit_03_r [{shapeType=Capsule, length=2.6153, radius=2.0364, center=<2.07 -1.5 0>, rotation=<rot 90 0 180>}]
 dyn_shirtBack_01 [{shapeType=Capsule, length=3.5602, radius=2.4003, center=<3 -1.2 0>, rotation=<rot 90 0 0>}]
 dyn_shirtBack_01_Base [{shapeType=Capsule, length=0.505, radius=0.505, rotation=<rot 0 -90 0>}]
 dyn_shirtBack_01_Base_l [{shapeType=Capsule, length=0.505, radius=0.505, rotation=<rot 0 -90 0>}]
 dyn_shirtBack_01_Base_r [{shapeType=Capsule, length=0.505, radius=0.505, rotation=<rot 0 -90 0>}]
 dyn_shirtBack_01_l [{shapeType=Capsule, length=3.6158, radius=2.6436, center=<3 -1.25 0>, rotation=<rot 90 0 0>}]
 dyn_shirtBack_01_r [{shapeType=Capsule, length=3.6337, radius=2.6988, center=<3 -1.25 0>, rotation=<rot 90 0 0>}]
 dyn_shirtBack_02 [{shapeType=Capsule, length=4.8749, radius=2.3711, center=<3 -1.2 0>, rotation=<rot 90 0 0>}]
 dyn_shirtBack_02_l [{shapeType=Capsule, length=3.7678, radius=2.6175, center=<3.5 -1.5 0>, rotation=<rot 90 0 0>}]
 dyn_shirtBack_02_r [{shapeType=Capsule, length=3.9378, radius=2.1893, center=<3.5 -1.5 0>, rotation=<rot 90 0 0>}]
 dyn_skirtA_01_Base_l [{shapeType=Box, extentX=1.01, extentY=1.01, extentZ=1.01, rotation=<rot 0 -90 0>}]
 dyn_skirtA_01_Base_r [{shapeType=Box, extentX=1.01, extentY=1.01, extentZ=1.01, rotation=<rot 0 -90 0>}]
 dyn_skirtA_01_l [{shapeType=Box, extentX=14.0356, extentY=2.8452, extentZ=11.7285, center=<6.4355 0 0>, rotation=<rot 0 0 -5.2002>}]
 dyn_skirtA_01_r [{shapeType=Box, extentX=14.0356, extentY=2.8452, extentZ=11.7285, center=<6.4355 0 0>, rotation=<rot 0 0 -5.2002>}]
 dyn_skirtA_02_l [{shapeType=Box, extentX=9.5772, extentY=2.73, extentZ=11.3161, center=<4.4383 -0.0002 0>, rotation=<rot 0 0 -1.0235>}]
 dyn_skirtA_02_r [{shapeType=Box, extentX=9.5772, extentY=2.73, extentZ=11.3161, center=<4.4383 -0.0002 0>, rotation=<rot 0 0 -1.0235>}]
 dyn_skirtA_03_l [{shapeType=Box, extentX=9.5772, extentY=2.73, extentZ=11.3161, center=<4.4383 -0.0002 0>, rotation=<rot -2.7206 0.0486 -1.0247>}]
 dyn_skirtA_03_r [{shapeType=Box, extentX=5.1779, extentY=2.73, extentZ=11.3161, center=<2.2968 -0.0001 0>, rotation=<rot 0 0 -1.0235>}]
 dyn_skirtB_01_Base_l [{shapeType=Box, extentX=1.01, extentY=1.01, extentZ=1.01, rotation=<rot 0 -90 0>}]
 dyn_skirtB_01_Base_r [{shapeType=Box, extentX=1.01, extentY=1.01, extentZ=1.01, rotation=<rot 0 -90 0>}]
 dyn_skirtB_01_l [{shapeType=Box, extentX=14.0356, extentY=2.8452, extentZ=11.7285, center=<6.4355 0 0>, rotation=<rot 0 0 -5.2002>}]
 dyn_skirtB_01_r [{shapeType=Box, extentX=14.0356, extentY=2.8452, extentZ=11.7285, center=<6.4355 0 0>, rotation=<rot 0 0 -5.2002>}]
 dyn_skirtB_02_l [{shapeType=Box, extentX=9.5772, extentY=2.73, extentZ=11.3161, center=<4.4383 -0.0002 0>, rotation=<rot 0 0 -16.4219>}]
 dyn_skirtB_02_r [{shapeType=Box, extentX=9.5772, extentY=2.73, extentZ=11.3161, center=<4.4383 -0.0002 0>, rotation=<rot 0 0 -1.0235>}]
 dyn_skirtB_03_l [{shapeType=Box, extentX=6.446, extentY=2.73, extentZ=11.3161, center=<3.0397 -1.2445 -0.3667>, rotation=<rot 0 0 -16.4227>}]
 dyn_skirtB_03_r [{shapeType=Box, extentX=6.6628, extentY=2.73, extentZ=11.3161, center=<2.7418 -0.0001 0>, rotation=<rot 0 0 -1.0235>}]
 dyn_skirtBack_01 [{shapeType=Box, extentX=11.2141, extentY=2.8452, extentZ=11.7285, center=<4.8348 -0.0002 0>, rotation=<rot 0 0 -5.2002>}]
 dyn_skirtBack_01_Base [{shapeType=Box, extentX=1.01, extentY=1.01, extentZ=1.01, rotation=<rot 0 -90 0>}]
 dyn_skirtBack_02 [{shapeType=Box, extentX=11.6435, extentY=2.73, extentZ=11.3161, center=<5.4165 -0.0001 0>, rotation=<rot 0 0 -1.0235>}]
 dyn_skirtBack_03 [{shapeType=Box, extentX=9.5772, extentY=2.73, extentZ=11.3161, center=<4.4383 -0.0002 0>, rotation=<rot 0 0 -1.0235>}]
 dyn_skirtBack_04 [{shapeType=Box, extentX=5.783, extentY=2.73, extentZ=11.3161, center=<2.2886 -0.0003 0>, rotation=<rot 0 0 -1.0235>}]
 dyn_skirtC_01_Base_l [{shapeType=Box, extentX=1.01, extentY=1.01, extentZ=1.01, rotation=<rot 0 -90 0>}]
 dyn_skirtC_01_l [{shapeType=Box, extentX=14.0356, extentY=2.8452, extentZ=11.7285, center=<6.4355 0 0>, rotation=<rot 0 0 -5.2002>}]
 dyn_skirtC_02_l [{shapeType=Box, extentX=11.61, extentY=2.73, extentZ=11.3161, center=<5.5472 -0.0002 0>, rotation=<rot 0 0 -1.0235>}]
 dyn_skirtD_01_Base_l [{shapeType=Box, extentX=1.01, extentY=1.01, extentZ=1.01, rotation=<rot 0 -90 0>}]
 dyn_skirtD_01_Base_r [{shapeType=Box, extentX=1.01, extentY=1.01, extentZ=1.01, rotation=<rot 0 -90 0>}]
 dyn_skirtD_01_l [{shapeType=Box, extentX=14.0356, extentY=2.8452, extentZ=11.7285, center=<6.4355 0 0>, rotation=<rot 0 0 -5.2002>}]
 dyn_skirtD_01_r [{shapeType=Box, extentX=14.0356, extentY=2.8452, extentZ=11.7285, center=<6.4355 0 0>, rotation=<rot 0 0 -5.2002>}]
 dyn_skirtD_02_l [{shapeType=Box, extentX=10.859, extentY=2.73, extentZ=11.3161, center=<5.321 -0.0002 0>, rotation=<rot 0 0 -1.0235>}]
 dyn_skirtD_02_r [{shapeType=Box, extentX=10.4575, extentY=2.73, extentZ=11.3161, center=<4.9111 -0.0002 0>, rotation=<rot 0 0 -1.0235>}]
 dyn_skirtD_03_l [{shapeType=Box, extentX=10.5512, extentY=3.73, extentZ=11.3161, center=<4.6716 -0.0002 0>, rotation=<rot 0 0 -1.0235>}]
 dyn_skirtD_03_r [{shapeType=Box, extentX=11.2927, extentY=2.73, extentZ=11.3161, center=<5.3659 -0.0001 0>, rotation=<rot 0 0 -1.0235>}]
 dyn_skirtD_04_l [{shapeType=Box, extentX=6.7742, extentY=3.73, extentZ=11.3161, center=<3.0373 -0.0002 0>, rotation=<rot 0 0 -1.0235>}]
 dyn_skirtD_04_r [{shapeType=Box, extentX=4.4356, extentY=2.73, extentZ=11.3161, center=<1.8383 -0.0002 0>, rotation=<rot 0 0 -1.0235>}]
 dyn_skirtSashBottomBase [{shapeType=Capsule, length=5.9818, radius=3.5076, center=<3.2782 -3.2516 0.0674>, rotation=<rot 0 -89.9339 -87.2772>}]
 dyn_skirtSashBottom_01 [{shapeType=Capsule, length=3.7863, radius=3.5076, center=<3.2782 -3.2516 0.0674>, rotation=<rot 0 -89.9339 -87.2772>}]
 dyn_skirtSashBottom_02 [{shapeType=Capsule, length=3.2822, radius=3.5076, center=<3.2782 -3.2516 0.0674>, rotation=<rot 0 -89.9339 -87.2772>}]
 dyn_skirtSashTop_01 [{shapeType=Capsule, length=6.0457, radius=3.5076, center=<4.6516 -3.2501 0.0021>, rotation=<rot 0 -89.9339 -87.2772>}]
 dyn_skirtSashTop_02 [{shapeType=Capsule, length=3.2822, radius=3.5076, center=<3.2771 -2.2639 0.0674>, rotation=<rot 0 -89.9339 -87.2772>}]
 dyn_wristDangleA_01_r [{shapeType=Capsule, length=5.3317, radius=0.5, center=<2.5579 -0.0095 0.0122>, rotation=<rot 0 -89.8359 -89.3107>}]
 dyn_wristDangleA_02_r [{shapeType=Capsule, length=5.624, radius=0.5318, center=<2.9735 -0.0048 -0.0272>, rotation=<rot 0 -90.5844 -89.6867>}]
 dyn_wristDangleB_01_r [{shapeType=Capsule, length=3.1184, radius=0.505, center=<1.3586 -0.0041 -0.0264>, rotation=<rot 0 -89.3033 -88.0189>}]
 dyn_wristDangleB_02_r [{shapeType=Capsule, length=3.6445, radius=0.505, center=<2.0069 0.03 -0.0131>, rotation=<rot 0 -90.5878 -89.1718>}]
 dyn_wristDangleMain__Base_r [{shapeType=Capsule, length=0.505, radius=0.505, rotation=<rot 0 -90 0>}]
 dyn_wristDangleMain__r [{shapeType=Capsule, length=3.1184, radius=0.505, center=<1.0713 0 0>, rotation=<rot 90 0 0>}]
 spine_03 [{shapeType=Capsule, length=11.6746, radius=3.0199, center=<3.134 9.8179 0>}]
 spine_04 [{shapeType=Capsule, length=13.8847, radius=10.5017, rotation=<rot 90 0 0>}]
 spine_05 [{shapeType=Capsule, length=14.5068, radius=11.4036, rotation=<rot 90 0 0>}]
 thigh_l [{shapeType=Capsule, length=36.4076, radius=6.3375, center=<-20.2182 -1.2481 -1.051>, rotation=<rot 82.1111 -179.9998 176.857>}]
 thigh_r [{shapeType=Capsule, length=36.0633, radius=6.6818, center=<22.3077 1.0473 0.8703>, rotation=<rot 86.555 -179.9998 0>}]
 thigh_twist_01_l [{shapeType=Capsule, length=15.0103, radius=10.3668, center=<5.4602 0.0004 0.5728>, rotation=<rot 84.0066 -180 -180>}]
}
constraints [
 {
  bone1Name=dyn_Base_l
  bone2Name=calf_l
  swing1LimitDegrees=45
  swing1Motion=Free
  swing2LimitDegrees=45
  swing2Motion=Free
  twistLimitDegrees=45
  twistMotion=Free
 }
 {bone1Name=dyn_legDangleMain_l, bone2Name=dyn_Base_l, swing1LimitDegrees=10, swing2LimitDegrees=8, twistLimitDegrees=4}
 {bone1Name=dyn_legDangleA_l, bone2Name=dyn_legDangleMain_l, swing1LimitDegrees=10, swing2LimitDegrees=8, twistLimitDegrees=4}
 {bone1Name=dyn_legDangleB_l, bone2Name=dyn_legDangleMain_l, swing1LimitDegrees=10, swing2LimitDegrees=8, twistLimitDegrees=4}
 {bone1Name=dyn_skirtA_01_l, bone2Name=dyn_skirtA_01_Base_l, swing1LimitDegrees=15, swing2LimitDegrees=15, twistLimitDegrees=6}
 {bone1Name=dyn_skirtA_02_l, bone2Name=dyn_skirtA_01_l, swing1LimitDegrees=15, swing2LimitDegrees=15, twistLimitDegrees=6}
 {bone1Name=dyn_skirtA_03_l, bone2Name=dyn_skirtA_02_l, swing1LimitDegrees=15, swing2LimitDegrees=15, twistLimitDegrees=6}
 {bone1Name=dyn_skirtB_01_l, bone2Name=dyn_skirtB_01_Base_l, swing1LimitDegrees=15, swing2LimitDegrees=15, twistLimitDegrees=6}
 {bone1Name=dyn_skirtB_02_l, bone2Name=dyn_skirtB_01_l, swing1LimitDegrees=15, swing2LimitDegrees=15, twistLimitDegrees=6}
 {bone1Name=dyn_skirtB_03_l, bone2Name=dyn_skirtB_02_l, swing1LimitDegrees=15, swing2LimitDegrees=15, twistLimitDegrees=6}
 {bone1Name=dyn_skirtC_01_l, bone2Name=dyn_skirtC_01_Base_l, swing1LimitDegrees=15, swing2LimitDegrees=15, twistLimitDegrees=6}
 {bone1Name=dyn_skirtC_02_l, bone2Name=dyn_skirtC_01_l, swing1LimitDegrees=15, swing2LimitDegrees=15, twistLimitDegrees=6}
 {bone1Name=dyn_skirtA_01_r, bone2Name=dyn_skirtA_01_Base_r, swing1LimitDegrees=15, swing2LimitDegrees=15, twistLimitDegrees=6}
 {bone1Name=dyn_skirtA_02_r, bone2Name=dyn_skirtA_01_r, swing1LimitDegrees=15, swing2LimitDegrees=15, twistLimitDegrees=6}
 {bone1Name=dyn_skirtA_03_r, bone2Name=dyn_skirtA_02_r, swing1LimitDegrees=15, swing2LimitDegrees=15, twistLimitDegrees=6}
 {bone1Name=dyn_skirtB_01_r, bone2Name=dyn_skirtB_01_Base_r, swing1LimitDegrees=15, swing2LimitDegrees=15, twistLimitDegrees=6}
 {bone1Name=dyn_skirtB_02_r, bone2Name=dyn_skirtB_01_r, swing1LimitDegrees=15, swing2LimitDegrees=15, twistLimitDegrees=6}
 {bone1Name=dyn_skirtB_03_r, bone2Name=dyn_skirtB_02_r, swing1LimitDegrees=15, swing2LimitDegrees=15, twistLimitDegrees=6}
 {bone1Name=dyn_skirtBack_01, bone2Name=dyn_skirtBack_01_Base, swing1LimitDegrees=20, swing2LimitDegrees=20, twistLimitDegrees=6}
 {bone1Name=dyn_skirtBack_02, bone2Name=dyn_skirtBack_01, swing1LimitDegrees=20, swing2LimitDegrees=20, twistLimitDegrees=6}
 {bone1Name=dyn_skirtBack_03, bone2Name=dyn_skirtBack_02, swing1LimitDegrees=20, swing2LimitDegrees=20, twistLimitDegrees=6}
 {bone1Name=dyn_skirtBack_04, bone2Name=dyn_skirtBack_03, swing1LimitDegrees=20, swing2LimitDegrees=20, twistLimitDegrees=6}
 {bone1Name=dyn_skirtD_01_r, bone2Name=dyn_skirtD_01_Base_r, swing1LimitDegrees=20, swing2LimitDegrees=20, twistLimitDegrees=6}
 {bone1Name=dyn_skirtD_02_r, bone2Name=dyn_skirtD_01_r, swing1LimitDegrees=20, swing2LimitDegrees=20, twistLimitDegrees=6}
 {bone1Name=dyn_skirtD_03_r, bone2Name=dyn_skirtD_02_r, swing1LimitDegrees=20, swing2LimitDegrees=20, twistLimitDegrees=6}
 {bone1Name=dyn_skirtD_04_r, bone2Name=dyn_skirtD_03_r, swing1LimitDegrees=20, swing2LimitDegrees=20, twistLimitDegrees=6}
 {bone1Name=dyn_skirtD_01_l, bone2Name=dyn_skirtD_01_Base_l, swing1LimitDegrees=20, swing2LimitDegrees=20, twistLimitDegrees=6}
 {bone1Name=dyn_skirtD_02_l, bone2Name=dyn_skirtD_01_l, swing1LimitDegrees=20, swing2LimitDegrees=20, twistLimitDegrees=6}
 {bone1Name=dyn_skirtD_03_l, bone2Name=dyn_skirtD_02_l, swing1LimitDegrees=20, swing2LimitDegrees=20, twistLimitDegrees=6}
 {bone1Name=dyn_skirtD_04_l, bone2Name=dyn_skirtD_03_l, swing1LimitDegrees=20, swing2LimitDegrees=20, twistLimitDegrees=6}
 {bone1Name=dyn_innerRope_01, bone2Name=dyn_skirtD_01_l, swing1LimitDegrees=6, swing2LimitDegrees=12, twistLimitDegrees=8}
 {bone1Name=dyn_innerRope_02, bone2Name=dyn_innerRope_01, swing1LimitDegrees=12, swing2LimitDegrees=12, twistLimitDegrees=2}
 {bone1Name=dyn_innerRope_03, bone2Name=dyn_innerRope_02, swing1LimitDegrees=12, swing2LimitDegrees=12, twistLimitDegrees=2}
 {bone1Name=dyn_backFlap_01_l, bone2Name=dyn_backFlap_01_Base_l, swing1LimitDegrees=25, swing2LimitDegrees=6, twistLimitDegrees=4}
 {bone1Name=dyn_backFlap_02_l, bone2Name=dyn_backFlap_01_l, swing1LimitDegrees=25, swing2LimitDegrees=6, twistLimitDegrees=4}
 {bone1Name=dyn_backFlapCenter_01, bone2Name=dyn_backFlapCenter_01_Base, swing1LimitDegrees=18, swing2LimitDegrees=10, twistLimitDegrees=4}
 {bone1Name=dyn_backFlapCenter_02, bone2Name=dyn_backFlapCenter_01, swing1LimitDegrees=20, swing2LimitDegrees=10, twistLimitDegrees=4}
 {bone1Name=dyn_outerRope_02, bone2Name=dyn_innerRope_02, swing1LimitDegrees=18, swing2LimitDegrees=12, twistLimitDegrees=2}
 {bone1Name=dyn_outerRope_03, bone2Name=dyn_outerRope_02, swing1LimitDegrees=12, swing2LimitDegrees=12, twistLimitDegrees=2}
 {bone1Name=dyn_forearmFlapA_01_l, bone2Name=dyn_forearmFlapA_01_Base_l, swing1LimitDegrees=20, swing2LimitDegrees=12, twistLimitDegrees=6}
 {bone1Name=dyn_forearmFlapB_01_l, bone2Name=dyn_forearmFlapB_01_Base_l, swing1LimitDegrees=20, swing2LimitDegrees=12, twistLimitDegrees=6}
 {bone1Name=dyn_forearmFlap_01_r, bone2Name=dyn_forearmFlap_01_Base_r, swing1LimitDegrees=14, swing2LimitDegrees=14, twistLimitDegrees=14}
 {bone1Name=dyn_wristDangleMain__r, bone2Name=dyn_wristDangleMain__Base_r, swing1LimitDegrees=25, swing2LimitDegrees=25, twistLimitDegrees=12}
 {bone1Name=dyn_wristDangleB_01_r, bone2Name=dyn_wristDangleMain__r, swing1LimitDegrees=14, swing2LimitDegrees=14, twistLimitDegrees=14}
 {bone1Name=dyn_wristDangleB_02_r, bone2Name=dyn_wristDangleB_01_r, swing1LimitDegrees=14, swing2LimitDegrees=14, twistLimitDegrees=14}
 {bone1Name=dyn_wristDangleA_01_r, bone2Name=dyn_wristDangleMain__r, swing1LimitDegrees=14, swing2LimitDegrees=14, twistLimitDegrees=14}
 {bone1Name=dyn_wristDangleA_02_r, bone2Name=dyn_wristDangleA_01_r, swing1LimitDegrees=14, swing2LimitDegrees=14, twistLimitDegrees=14}
 {bone1Name=dyn_hipSash_02, bone2Name=dyn_hipSash_01, swing1LimitDegrees=16, swing2LimitDegrees=12, twistLimitDegrees=6}
 {bone1Name=dyn_hipSash_03, bone2Name=dyn_hipSash_02, swing1LimitDegrees=16, swing2LimitDegrees=12, twistLimitDegrees=6}
 {bone1Name=dyn_skirtSashBottomBase, bone2Name=dyn_hipSash_03, swing1LimitDegrees=16, swing2LimitDegrees=12, twistLimitDegrees=6}
 {bone1Name=dyn_skirtSashBottom_01, bone2Name=dyn_skirtSashBottomBase, swing1LimitDegrees=16, swing2LimitDegrees=12, twistLimitDegrees=6}
 {bone1Name=dyn_skirtSashBottom_02, bone2Name=dyn_skirtSashBottom_01, swing1LimitDegrees=16, swing2LimitDegrees=12, twistLimitDegrees=6}
 {bone1Name=dyn_skirtSashTop_01, bone2Name=dyn_hipSash_03, swing1LimitDegrees=16, swing2LimitDegrees=12, twistLimitDegrees=6}
 {bone1Name=dyn_skirtSashTop_02, bone2Name=dyn_skirtSashTop_01, swing1LimitDegrees=16, swing2LimitDegrees=12, twistLimitDegrees=6}
 {bone1Name=dyn_pouchA_01, bone2Name=dyn_pouchA_01_Base, swing1LimitDegrees=10, swing2LimitDegrees=8, twistLimitDegrees=4}
 {bone1Name=dyn_pouchB_01, bone2Name=dyn_pouchB_01_Base, swing1LimitDegrees=10, swing2LimitDegrees=4, twistLimitDegrees=4}
 {bone1Name=dyn_pouchC_01, bone2Name=dyn_pouchC_01_Base, swing1LimitDegrees=10, swing2LimitDegrees=4, twistLimitDegrees=4}
 {bone1Name=dyn_quiverMain, bone2Name=dyn_quiverMain_Base, swing1LimitDegrees=20, swing2LimitDegrees=20, twistLimitDegrees=4}
 {
  bone1Name=calf_l
  bone2Name=thigh_l
  swing1LimitDegrees=45
  swing1Motion=Free
  swing2LimitDegrees=45
  swing2Motion=Free
  twistLimitDegrees=45
  twistMotion=Free
 }
 {
  bone1Name=thigh_twist_01_l
  bone2Name=thigh_l
  swing1LimitDegrees=45
  swing1Motion=Free
  swing2LimitDegrees=45
  swing2Motion=Free
  twistLimitDegrees=45
  twistMotion=Free
 }
 {
  bone1Name=dyn_skirtA_01_Base_l
  bone2Name=thigh_l
  swing1LimitDegrees=45
  swing1Motion=Free
  swing2LimitDegrees=45
  swing2Motion=Free
  twistLimitDegrees=45
  twistMotion=Free
 }
 {
  bone1Name=dyn_skirtB_01_Base_l
  bone2Name=thigh_l
  swing1LimitDegrees=45
  swing1Motion=Free
  swing2LimitDegrees=45
  swing2Motion=Free
  twistLimitDegrees=45
  twistMotion=Free
 }
 {
  bone1Name=dyn_skirtC_01_Base_l
  bone2Name=thigh_l
  swing1LimitDegrees=45
  swing1Motion=Free
  swing2LimitDegrees=45
  swing2Motion=Free
  twistLimitDegrees=45
  twistMotion=Free
 }
 {bone1Name=dyn_shirtArmpit_01_l, bone2Name=dyn_shirtArmpit_01_Base_l, swing1LimitDegrees=4, swing2LimitDegrees=12, twistLimitDegrees=4}
 {bone1Name=dyn_shirtArmpit_02_l, bone2Name=dyn_shirtArmpit_01_l, swing1LimitDegrees=5, swing2LimitDegrees=5, twistLimitDegrees=4}
 {bone1Name=dyn_shirtArmpit_03_l, bone2Name=dyn_shirtArmpit_02_l, swing1LimitDegrees=5, swing2LimitDegrees=5, twistLimitDegrees=4}
 {bone1Name=dyn_shirtArmpit_01_r, bone2Name=dyn_shirtArmpit_01_Base_r, swing1LimitDegrees=4, swing2LimitDegrees=12, twistLimitDegrees=4}
 {bone1Name=dyn_shirtArmpit_02_r, bone2Name=dyn_shirtArmpit_01_r, swing1LimitDegrees=5, swing2LimitDegrees=5, twistLimitDegrees=4}
 {bone1Name=dyn_shirtArmpit_03_r, bone2Name=dyn_shirtArmpit_02_r, swing1LimitDegrees=5, swing2LimitDegrees=5, twistLimitDegrees=4}
 {bone1Name=dyn_shirtBack_01_r, bone2Name=dyn_shirtBack_01_Base_r, swing1LimitDegrees=12, swing2LimitDegrees=8, twistLimitDegrees=4}
 {bone1Name=dyn_shirtBack_02_r, bone2Name=dyn_shirtBack_01_r, swing1LimitDegrees=10, swing2LimitDegrees=10, twistLimitDegrees=6}
 {bone1Name=dyn_shirtBack_01, bone2Name=dyn_shirtBack_01_Base, swing1LimitDegrees=12, swing2LimitDegrees=8, twistLimitDegrees=4}
 {bone1Name=dyn_shirtBack_02, bone2Name=dyn_shirtBack_01, swing1LimitDegrees=10, swing2LimitDegrees=10, twistLimitDegrees=6}
 {bone1Name=dyn_shirtBack_01_l, bone2Name=dyn_shirtBack_01_Base_l, swing1LimitDegrees=12, swing2LimitDegrees=8, twistLimitDegrees=4}
 {bone1Name=dyn_shirtBack_02_l, bone2Name=dyn_shirtBack_01_l, swing1LimitDegrees=10, swing2LimitDegrees=10, twistLimitDegrees=6}
 {
  bone1Name=dyn_forearmFlapA_01_Base_l
  bone2Name=spine_05
  swing1LimitDegrees=45
  swing1Motion=Free
  swing2LimitDegrees=45
  swing2Motion=Free
  twistLimitDegrees=45
  twistMotion=Free
 }
 {
  bone1Name=dyn_forearmFlapB_01_Base_l
  bone2Name=spine_05
  swing1LimitDegrees=45
  swing1Motion=Free
  swing2LimitDegrees=45
  swing2Motion=Free
  twistLimitDegrees=45
  twistMotion=Free
 }
 {
  bone1Name=dyn_forearmFlap_01_Base_r
  bone2Name=spine_05
  swing1LimitDegrees=45
  swing1Motion=Free
  swing2LimitDegrees=45
  swing2Motion=Free
  twistLimitDegrees=45
  twistMotion=Free
 }
 {
  bone1Name=dyn_wristDangleMain__Base_r
  bone2Name=spine_05
  swing1LimitDegrees=45
  swing1Motion=Free
  swing2LimitDegrees=45
  swing2Motion=Free
  twistLimitDegrees=45
  twistMotion=Free
 }
 {
  bone1Name=dyn_shirtArmpit_01_Base_l
  bone2Name=spine_05
  swing1LimitDegrees=45
  swing1Motion=Free
  swing2LimitDegrees=45
  swing2Motion=Free
  twistLimitDegrees=45
  twistMotion=Free
 }
 {
  bone1Name=dyn_shirtBack_01_Base_l
  bone2Name=spine_05
  swing1LimitDegrees=45
  swing1Motion=Free
  swing2LimitDegrees=45
  swing2Motion=Free
  twistLimitDegrees=45
  twistMotion=Free
 }
 {
  bone1Name=dyn_shirtArmpit_01_Base_r
  bone2Name=spine_05
  swing1LimitDegrees=45
  swing1Motion=Free
  swing2LimitDegrees=45
  swing2Motion=Free
  twistLimitDegrees=45
  twistMotion=Free
 }
 {
  bone1Name=dyn_shirtBack_01_Base_r
  bone2Name=spine_05
  swing1LimitDegrees=45
  swing1Motion=Free
  swing2LimitDegrees=45
  swing2Motion=Free
  twistLimitDegrees=45
  twistMotion=Free
 }
 {
  bone1Name=dyn_shirtBack_01_Base
  bone2Name=spine_05
  swing1LimitDegrees=45
  swing1Motion=Free
  swing2LimitDegrees=45
  swing2Motion=Free
  twistLimitDegrees=45
  twistMotion=Free
 }
 {
  bone1Name=spine_05
  bone2Name=spine_04
  swing1LimitDegrees=45
  swing1Motion=Free
  swing2LimitDegrees=45
  swing2Motion=Free
  twistLimitDegrees=45
  twistMotion=Free
 }
 {
  bone1Name=spine_04
  bone2Name=spine_03
  swing1LimitDegrees=45
  swing1Motion=Free
  swing2LimitDegrees=45
  swing2Motion=Free
  twistLimitDegrees=45
  twistMotion=Free
 }
 {
  bone1Name=calf_r
  bone2Name=thigh_r
  swing1LimitDegrees=45
  swing1Motion=Free
  swing2LimitDegrees=45
  swing2Motion=Free
  twistLimitDegrees=45
  twistMotion=Free
 }
 {
  bone1Name=dyn_skirtA_01_Base_r
  bone2Name=thigh_r
  swing1LimitDegrees=45
  swing1Motion=Free
  swing2LimitDegrees=45
  swing2Motion=Free
  twistLimitDegrees=45
  twistMotion=Free
 }
 {
  bone1Name=dyn_skirtB_01_Base_r
  bone2Name=thigh_r
  swing1LimitDegrees=45
  swing1Motion=Free
  swing2LimitDegrees=45
  swing2Motion=Free
  twistLimitDegrees=45
  twistMotion=Free
 }
]
