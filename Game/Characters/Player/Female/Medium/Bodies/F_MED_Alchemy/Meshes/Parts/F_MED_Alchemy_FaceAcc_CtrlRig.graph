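# Control Rig for the F_MED_Alchemy FaceAcc mesh that procedurally orients the character's back hair and front locks as the head moves.
# It places five aim-target bones under spine_05 and the clavicles, then offsets them using clamped remaps of the head's Euler rotation relative to spine_05.
# Multiplied weights add correctives for combined head rotations, including one on the right back-side base bone.
# Five AimBone units then point the hair base bones at their targets, using the head as secondary target.
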ControlRigBlueprint /Game/Characters/Player/Female/Medium/Bodies/F_MED_Alchemy/Meshes/Parts/F_MED_Alchemy_FaceAcc_CtrlRig
graphs [
 RigVMFunctionLibrary {links [], nodes []}
 RigVMModel {nodes [
   RigUnit_BeginExecution = &RigUnit_BeginExecution()
   &RigUnit_SetTransform(
    Item=(Type=Bone,Name="dyn_hairBackCenterAimTarget")
    Space=LocalSpace
    Value=(Rotation=(X=0,Y=0,Z=0,W=1),Translation=(X=5,Y=-22.75,Z=-0.25),Scale3D=(X=1,Y=1,Z=1))
    Weight=1
    bInitial=true
    ExecutePin=@RigUnit_BeginExecution.ExecutePin
   )
   > &RigUnit_SetTransform(
    Item=(Type=Bone,Name="dyn_hairBackLeftAimTarget")
    Space=LocalSpace
    Value=(Rotation=(X=0,Y=0,Z=0,W=1),Translation=(X=4.25,Y=-10,Z=0.5),Scale3D=(X=1,Y=1,Z=1))
    Weight=1
    bInitial=true
   )
   > &RigUnit_SetTransform(
    Item=(Type=Bone,Name="dyn_hairBackRightAimTarget")
    Space=LocalSpace
    Value=(Rotation=(X=0,Y=0,Z=0,W=1),Translation=(X=-4.25,Y=10,Z=-0.5),Scale3D=(X=1,Y=1,Z=1))
    Weight=1
    bInitial=true
   )
   > &RigUnit_SetTransform(
    Item=(Type=Bone,Name="dyn_hairFrontRightAimTarget")
    Space=LocalSpace
    Value=(Rotation=(X=0,Y=0,Z=0,W=1),Translation=(X=-3.5,Y=-5.4,Z=0),Scale3D=(X=1,Y=1,Z=1))
    Weight=1
    bInitial=true
   )
   > &RigUnit_SetTransform(
    Item=(Type=Bone,Name="dyn_hairFrontLeftAimTarget")
    Space=LocalSpace
    Value=(Rotation=(X=0,Y=0,Z=0,W=1),Translation=(X=3.6,Y=3.25,Z=0),Scale3D=(X=1,Y=1,Z=1))
    Weight=1
    bInitial=true
   )
   n1 = &RigVMFunction_MathQuaternionToEuler(
    RotationOrder=ZYX
    Value=&RigUnit_GetRelativeTransformForItem(Child=(Type=Bone,Name="head"), Parent=(Type=Bone,Name="spine_05")).RelativeTransform.Rotation
   )
   n2 = &RigVMFunction_MathFloatRemap(SourceMaximum=-56, TargetMinimum=0, bClamp=true, Value=@n1.Result.X)
   > /Script/ControlRig.RigUnit_ModifyBoneTransforms(
    BoneToModify="((Bone=\"dyn_hairBackCenterAimTarget\",Transform=(Rotation=(X=0,Y=0,Z=0,W=1),Translation=(X=0,Y=0,Z=10),Scale3D=(X=1,Y=1,Z=1))),(Bone=\"dyn_hairBackLeftAimTarget\",Transform=(Rotation=(X=0,Y=0,Z=0,W=1),Translation=(X=-2,Y=0,Z=0),Scale3D=(X=1,Y=1,Z=1))),(Bone=\" ...(truncated)"
    Mode=AdditiveLocal
    WeightMaximum=1
    Weight=@n2.Result
   ) {is=struct RigVMUnitNode}
   > &RigUnit_ModifyBoneTransforms(
    BoneToModify="((Bone=\"dyn_hairBackCenterAimTarget\",Transform=(Rotation=(X=0,Y=0,Z=0,W=1),Translation=(X=0,Y=0,Z=-10),Scale3D=(X=1,Y=1,Z=1))),(Bone=\"dyn_hairBackLeftAimTarget\",Transform=(Rotation=(X=0,Y=0,Z=0,W=1),Translation=(X=1.5,Y=0,Z=0),Scale3D=(X=1,Y=1,Z=1))),(Bone ...(truncated)"
    Mode=AdditiveLocal
    WeightMaximum=1
    Weight=&RigVMFunction_MathFloatRemap(SourceMaximum=25, TargetMinimum=0, bClamp=true, Value=@n1.Result.X).Result
   )
   n3 = &RigVMFunction_MathFloatRemap(SourceMaximum=-43, SourceMinimum=0, TargetMinimum=0, bClamp=true, Value=@n1.Result.Z)
   > &RigUnit_ModifyTransforms(
    ItemToModify="((Item=(Type=Bone,Name=\"dyn_hairBackCenterAimTarget\"),Transform=(Rotation=(X=0,Y=0,Z=0,W=1),Translation=(X=-10,Y=0,Z=0),Scale3D=(X=1,Y=1,Z=1))),(Item=(Type=Bone,Name=\"dyn_hairBackLeftAimTarget\"),Transform=(Rotation=(X=0,Y=0,Z=0,W=1),Translation=(X=0,Y=0,Z= ...(truncated)"
    Weight=@n3.Result
   )
   > &RigUnit_ModifyTransforms(
    ItemToModify="((Item=(Type=Bone,Name=\"dyn_hairFrontLeftAimTarget\"),Transform=(Rotation=(X=0,Y=0,Z=0,W=1),Translation=(X=0,Y=0,Z=0),Scale3D=(X=1,Y=1,Z=1))),(Item=(Type=Bone,Name=\"dyn_hairFrontRightAimTarget\"),Transform=(Rotation=(X=0,Y=0,Z=0,W=1),Translation=(X=0,Y=0,Z=0 ...(truncated)"
    Weight=@n3.Result
   )
   n4 = &RigVMFunction_MathFloatMul(A=@n3.Result, B=&RigVMFunction_MathFloatRemap(SourceMaximum=0.6, bClamp=true, Value=@n2.Result).Result) {template="Multiply::Execute(in A,in B,out Result)"}
   > &RigUnit_ModifyBoneTransforms(
    BoneToModify="((Bone=\"dyn_hairBackCenterAimTarget\",Transform=(Rotation=(X=0,Y=0,Z=0,W=1),Translation=(X=0,Y=0,Z=0),Scale3D=(X=1,Y=1,Z=1))),(Bone=\"dyn_hairBackLeftAimTarget\",Transform=(Rotation=(X=0,Y=0,Z=0,W=1),Translation=(X=0,Y=0,Z=0),Scale3D=(X=1,Y=1,Z=1))),(Bone=\"dy ...(truncated)"
    Mode=AdditiveLocal
    WeightMaximum=1
    Weight=@n4.Result
    BoneToModify.2.Transform.Rotation=&RigVMFunction_MathQuaternionFromEuler(Euler=(X=-10,Y=-5,Z=-15)).Result
   )
   > &RigUnit_ModifyTransforms(
    ItemToModify=((Item=(Type=Bone,Name="dyn_hairBackSide_01_Base_r"),Transform=(Rotation=(X=0,Y=0,Z=0,W=1),Translation=(X=-2,Y=-1.5,Z=0),Scale3D=(X=1,Y=1,Z=1))))
    Weight=&RigVMFunction_MathFloatMul(
     A=@n4.Result
     B=&RigVMFunction_MathFloatRemap(SourceMaximum=-28, SourceMinimum=0, TargetMinimum=0, bClamp=true, Value=@n1.Result.Y).Result
    ) {template="Multiply::Execute(in A,in B,out Result)"}.Result
    ItemToModify.0.Transform.Rotation=&RigVMFunction_MathQuaternionFromEuler(Euler=(X=-15,Y=-15,Z=0)).Result
   )
   > /Script/ControlRig.RigUnit_AimBone(
    Bone=dyn_hairBackMain_01_Base
    DebugSettings=(Scale=10,WorldOffset=(Rotation=(X=0,Y=0,Z=0,W=1),Translation=(X=0,Y=0,Z=0),Scale3D=(X=1,Y=1,Z=1)))
    Primary=(Weight=1,Axis=(X=1,Y=0,Z=0),Target=(X=1,Y=0,Z=0),Kind=Location,Space="None")
    Secondary=(Weight=1,Axis=(X=0,Y=1,Z=0),Target=(X=1,Y=0,Z=0),Kind=Location)
    Weight=1
    bPropagateToChildren=True
    Secondary.Target=&RigUnit_GetBoneTransform(Bone=head).Transform.Translation
    Primary.Target=&RigUnit_GetBoneTransform(Bone=dyn_hairBackCenterAimTarget).Transform.Translation
   ) {is=struct RigVMUnitNode}
   > &RigUnit_AimBone(
    Bone=dyn_hairBackSide_01_Base_l
    DebugSettings=(Scale=10,WorldOffset=(Rotation=(X=0,Y=0,Z=0,W=1),Translation=(X=0,Y=0,Z=0),Scale3D=(X=1,Y=1,Z=1)))
    Primary=(Weight=1,Axis=(X=1,Y=0,Z=0),Target=(X=1,Y=0,Z=0),Kind=Location,Space="None")
    Secondary=(Weight=1,Axis=(X=0,Y=1,Z=0),Target=(X=1,Y=0,Z=0),Kind=Location)
    Weight=1
    bPropagateToChildren=True
    Primary.Target=&RigUnit_GetBoneTransform(Bone=dyn_hairBackLeftAimTarget).Transform.Translation
    Secondary.Target=&RigUnit_GetBoneTransform(Bone=head).Transform.Translation
   )
   > &RigUnit_AimBone(
    Bone=dyn_hairBackSide_01_Base_r
    DebugSettings=(Scale=10,WorldOffset=(Rotation=(X=0,Y=0,Z=0,W=1),Translation=(X=0,Y=0,Z=0),Scale3D=(X=1,Y=1,Z=1)))
    Primary=(Weight=1,Axis=(X=1,Y=0,Z=0),Target=(X=1,Y=0,Z=0),Kind=Location,Space="None")
    Secondary=(Weight=1,Axis=(X=0,Y=1,Z=0),Target=(X=1,Y=0,Z=0),Kind=Location)
    Weight=1
    bPropagateToChildren=True
    Primary.Target=&RigUnit_GetBoneTransform(Bone=dyn_hairBackRightAimTarget).Transform.Translation
    Secondary.Target=&RigUnit_GetBoneTransform(Bone=head).Transform.Translation
   )
   > &RigUnit_AimBone(
    Bone=dyn_frontLockMain_01_Base_r
    DebugSettings=(Scale=10,WorldOffset=(Rotation=(X=0,Y=0,Z=0,W=1),Translation=(X=0,Y=0,Z=0),Scale3D=(X=1,Y=1,Z=1)))
    Primary=(Weight=1,Axis=(X=1,Y=0,Z=0),Target=(X=1,Y=0,Z=0),Kind=Location,Space="None")
    Secondary=(Weight=1,Axis=(X=0,Y=-0.36,Z=-1),Target=(X=1,Y=0,Z=0),Kind=Location)
    Weight=1
    bPropagateToChildren=True
    Primary.Target=&RigUnit_GetBoneTransform(Bone=dyn_hairFrontRightAimTarget).Transform.Translation
    Secondary.Target=&RigUnit_GetBoneTransform(Bone=head).Transform.Translation
   )
   > &RigUnit_AimBone(
    Bone=dyn_frontLockMain_01_Base_l
    DebugSettings=(Scale=10,WorldOffset=(Rotation=(X=0,Y=0,Z=0,W=1),Translation=(X=0,Y=0,Z=0),Scale3D=(X=1,Y=1,Z=1)))
    Primary=(Weight=1,Axis=(X=1,Y=0,Z=0),Target=(X=1,Y=0,Z=0),Kind=Location,Space="None")
    Secondary=(Weight=1,Axis=(X=0,Y=0,Z=1),Target=(X=1,Y=0,Z=0),Kind=Location)
    Weight=1
    bPropagateToChildren=True
    Primary.Target=&RigUnit_GetBoneTransform(Bone=dyn_hairFrontLeftAimTarget).Transform.Translation
    Secondary.Target=&RigUnit_GetBoneTransform(Bone=head).Transform.Translation
   )
   &RigUnit_GetRelativeBoneTransform(Bone=head, Space=spine_05)
  ]}
]
hierarchy [
 root {loc=<0 0 0>, rot=<0 0 0 1>}
 pelvis {parent=root, loc=<0 -0.5301 98.0571>, rot=<0.0127 0.707 -0.0127 -0.707>}
 spine_01 {parent=pelvis, loc=<6.1208 -0.9328 0>, rot=<0 0 0.034 -0.9994>}
 spine_02 {parent=spine_01, loc=<6.5996 -0.2095 0>, rot=<0 0 0.0238 -0.9997>}
 spine_03 {parent=spine_02, loc=<6.4914 -0.1237 0>, rot=<0 0 0.0149 -0.9999>}
 spine_04 {parent=spine_03, loc=<6.9598 0.0794 0>, rot=<0 0 -0.0132 -0.9999>}
 spine_05 {parent=spine_04, loc=<6.9987 0 0>, rot=<0 0 -0.0015 -1>}
 neck_01 {parent=spine_05, loc=<11.3138 0.1529 0>, rot=<0 0 -0.2003 -0.9797>}
 neck_02 {parent=neck_01, loc=<4.5627 0 0>, rot=<0 0 0.0504 -0.9987>}
 head {parent=neck_02, loc=<5.3914 0 0>, rot=<0 0 0.115 -0.9934>}
 dyn_hairBackSide_01_Base_r {parent=head, loc=<2.8416 -7.8325 5.7051>, rot=<0.1367 -0.8967 0.4204 -0.0249>}
 dyn_hairBackSide_01_r {parent=dyn_hairBackSide_01_Base_r, loc=<0 0 0>, rot=<0 0 0 1>}
 dyn_hairBackSide_02_r {parent=dyn_hairBackSide_01_r, loc=<7.9819 0 0>, rot=<-0.0815 0.0275 0.2206 -0.9716>}
 dyn_hairBackSide_03_r {parent=dyn_hairBackSide_02_r, loc=<6.6883 0 0>, rot=<0.0399 -0.0222 -0.2242 -0.9735>}
 dyn_hairBackSide_04_r {parent=dyn_hairBackSide_03_r, loc=<7.9847 0 0>, rot=<-0.1436 -0.1227 -0.1677 -0.9676>}
 dyn_hairBackSide_05_r {parent=dyn_hairBackSide_04_r, loc=<4.2224 0 0>, rot=<-0.0597 -0.1 0.1306 -0.9846>}
 dyn_hairBackSide_01_Base_l {parent=head, loc=<2.8414 -7.8325 -5.7051>, rot=<0.1355 -0.8775 -0.4591 0.0309>}
 dyn_hairBackSide_01_l {parent=dyn_hairBackSide_01_Base_l, loc=<0 0 0>, rot=<0 0 0 -1>}
 dyn_hairBackSide_02_l {parent=dyn_hairBackSide_01_l, loc=<7.9819 0 0>, rot=<-0.0885 -0.0082 0.2221 -0.971>}
 dyn_hairBackSide_03_l {parent=dyn_hairBackSide_02_l, loc=<6.6883 0 0>, rot=<0.0457 -0.0835 -0.1511 -0.9839>}
 dyn_hairBackSide_04_l {parent=dyn_hairBackSide_03_l, loc=<8.2649 0 0>, rot=<0.1441 0.225 -0.3654 -0.8917>}
 dyn_hairBackSide_05_l {parent=dyn_hairBackSide_04_l, loc=<7.4431 0 0>, rot=<0.0282 0.2071 0.2291 -0.9507>}
 dyn_hairBackMain_01_Base {parent=head, loc=<6.9942 -9.1642 -0.7675>, rot=<0.2835 -0.9588 0.017 0.0049>}
 dyn_hairBackMain_01 {parent=dyn_hairBackMain_01_Base, loc=<0 0 0>, rot=<0 0 0 -1>}
 dyn_hairBackMain_02 {parent=dyn_hairBackMain_01, loc=<6.6523 0 0>, rot=<-0.0258 -0.0161 -0.0073 -0.9995>}
 dyn_hairBackMain_03 {parent=dyn_hairBackMain_02, loc=<10.5178 0 0>, rot=<0.0524 0.0338 -0.27 -0.9608>}
 dyn_hairBackMain_04 {parent=dyn_hairBackMain_03, loc=<11.5706 0 0>, rot=<0.0337 0.0378 -0.1006 -0.9936>}
 dyn_hairBackMain_05 {parent=dyn_hairBackMain_04, loc=<7.1231 0 0>, rot=<-0.0073 -0.0196 -0.0879 -0.9959>}
 dyn_hairBackMain_06 {parent=dyn_hairBackMain_05, loc=<5.7507 0 0>, rot=<-0.0049 0.2763 -0.1577 -0.948>}
 dyn_hairBackBraid_01_l {parent=dyn_hairBackMain_02, loc=<3.0451 0.0589 4.77>, rot=<0.0422 0.026 -0.0007 -0.9988>}
 dyn_hairBackBraid_02_l {parent=dyn_hairBackBraid_01_l, loc=<7.2439 0 0>, rot=<-0.0518 -0.0259 -0.0869 -0.9945>}
 dyn_hairBackBraid_03_l {parent=dyn_hairBackBraid_02_l, loc=<6.3608 0 0>, rot=<0.0758 0.0303 -0.1987 -0.9766>}
 dyn_hairBackBraid_04_l {parent=dyn_hairBackBraid_03_l, loc=<7.3346 0 0>, rot=<-0.0487 -0.0118 -0.0735 -0.996>}
 dyn_hairBackBraid_01_r {parent=dyn_hairBackMain_02, loc=<3.2283 -0.6224 -4.7775>, rot=<-0.0486 -0.0291 -0.013 -0.9983>}
 dyn_hairBackBraid_02_r {parent=dyn_hairBackBraid_01_r, loc=<7.4712 0 0>, rot=<0.0631 0.0255 -0.1408 -0.9877>}
 dyn_hairBackBraid_03_r {parent=dyn_hairBackBraid_02_r, loc=<7.3833 0 0>, rot=<-0.0191 0.027 -0.15 -0.9881>}
 dyn_hairBackBraid_04_r {parent=dyn_hairBackBraid_03_r, loc=<5.0945 0 0>, rot=<-0.1875 -0.027 0.0578 -0.9802>}
 dyn_frontLockMain_01_Base_r {parent=head, loc=<1.9373 -4.5789 5.6478>, rot=<0.0212 -0.102 0.9754 0.1943>}
 dyn_frontLockMain_01_r {parent=dyn_frontLockMain_01_Base_r, loc=<0 0 0>, rot=<0 0 0 1>}
 dyn_frontLockMain_02_r {parent=dyn_frontLockMain_01_r, loc=<8.468 0 0>, rot=<-0.0023 0.0425 0.2585 -0.9651>}
 dyn_frontLockMain_03_r {parent=dyn_frontLockMain_02_r, loc=<7.58 0 0>, rot=<0.0871 0.1432 -0.0809 -0.9825>}
 dyn_frontLockMain_04_r {parent=dyn_frontLockMain_03_r, loc=<6.5705 0 0>, rot=<-0.041 -0.1422 -0.1112 -0.9827>}
 dyn_frontLockMain_05_r {parent=dyn_frontLockMain_04_r, loc=<6.6247 0 0>, rot=<0.0143 0.0338 -0.072 -0.9967>}
 dyn_frontLockB_01_r {parent=dyn_frontLockMain_04_r, loc=<6.8812 1.1432 -2.1622>, rot=<-0.1782 -0.0561 -0.1344 -0.9732>}
 dyn_frontLockB_02_r {parent=dyn_frontLockB_01_r, loc=<2.7454 0 0>, rot=<-0.1702 -0.0583 0.0352 -0.983>}
 dyn_frontLockA_01_r {parent=dyn_frontLockMain_03_r, loc=<3.1545 0.1102 2.2062>, rot=<0.0789 -0.0167 -0.1194 -0.9896>}
 dyn_frontLockA_02_r {parent=dyn_frontLockA_01_r, loc=<2.3387 0 0>, rot=<-0.0737 -0.0441 -0.0214 -0.9961>}
 dyn_frontLockA_03_r {parent=dyn_frontLockA_02_r, loc=<4.2629 0 0>, rot=<0.1164 0.0526 -0.0931 -0.9874>}
 dyn_frontLockC_01_r {parent=dyn_frontLockMain_03_r, loc=<2.6902 1.1099 -3.3941>, rot=<-0.0656 -0.1433 -0.0472 -0.9864>}
 dyn_frontLockC_02_r {parent=dyn_frontLockC_01_r, loc=<4.3221 0 0>, rot=<-0.1118 -0.0811 -0.053 -0.989>}
 dyn_frontLockC_03_r {parent=dyn_frontLockC_02_r, loc=<1.7969 0 0>, rot=<0.0788 0.0457 -0.0484 -0.9947>}
 dyn_frontLockMain_01_Base_l {parent=head, loc=<2.2448 -2.9595 -5.6196>, rot=<-0.0229 0.2047 0.9733 0.1009>}
 dyn_frontLockMain_01_l {parent=dyn_frontLockMain_01_Base_l, loc=<0 0 0>, rot=<0 0 0 1>}
 dyn_frontLockMain_02_l {parent=dyn_frontLockMain_01_l, loc=<8.3067 0 0>, rot=<0.0442 0.0284 0.3488 -0.9357>}
 dyn_frontLockMain_03_l {parent=dyn_frontLockMain_02_l, loc=<7.5802 0 0>, rot=<-0.0441 -0.1466 -0.0746 -0.9854>}
 dyn_frontLockMain_04_l {parent=dyn_frontLockMain_03_l, loc=<6.5701 0 0>, rot=<0.0801 0.0341 -0.2135 -0.9731>}
 dyn_frontLockMain_05_l {parent=dyn_frontLockMain_04_l, loc=<8.8262 0 0>, rot=<-0.0854 0.1734 -0.0492 -0.9799>}
 dyn_frontLockB_01_l {parent=dyn_frontLockMain_04_l, loc=<4.2402 -0.5543 1.8708>, rot=<0.0009 0.149 0.0103 -0.9888>}
 dyn_frontLockB_02_l {parent=dyn_frontLockB_01_l, loc=<2.762 0 0>, rot=<0.0359 -0.0133 -0.07 -0.9968>}
 dyn_frontLockA_01_l {parent=dyn_frontLockMain_03_l, loc=<2.2467 -0.6074 -3.0886>, rot=<-0.1163 -0.0152 -0.1102 -0.987>}
 dyn_frontLockA_02_l {parent=dyn_frontLockA_01_l, loc=<5.191 0 0>, rot=<-0.058 0.0167 -0.1378 -0.9886>}
 dyn_frontLockA_03_l {parent=dyn_frontLockA_02_l, loc=<5.6424 0 0>, rot=<-0.0491 0.0778 -0.1451 -0.9851>}
 dyn_frontLockC_01_l {parent=dyn_frontLockMain_03_l, loc=<6.7168 0.5378 2.8988>, rot=<0.1192 0.2049 -0.1651 -0.9574>}
 dyn_hairBackCenterAimTarget {parent=spine_05, loc=<0 0 0>, rot=<0 0 0 1>}
 clavicle_l {parent=spine_05, loc=<5.4831 -0.1053 -3.2946>, rot=<0.0465 -0.7289 -0.009 -0.6829>}
 dyn_hairBackLeftAimTarget {parent=clavicle_l, loc=<0 0 0>, rot=<0 0 0 1>}
 dyn_hairFrontLeftAimTarget {parent=clavicle_l, loc=<0 0 0>, rot=<0 0 0 1>}
 clavicle_r {parent=spine_05, loc=<5.4831 -0.1053 3.2946>, rot=<0.7289 0.0465 -0.6829 0.009>}
 dyn_hairBackRightAimTarget {parent=clavicle_r, loc=<0 0 0>, rot=<0 0 0 1>}
 dyn_hairFrontRightAimTarget {parent=clavicle_r, loc=<0 0 0>, rot=<0 0 0 1>}
]
