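# Physics asset for the hair of the F_MED_Alchemy female character, named as a back bling variant.
# Capsules and a sphere cover upper spine, clavicles, neck and head as the body volume around the hair.
# Three back chains and two front lock chains with smaller branching strands hang from the head via free-rotation base joints, mostly with tight angular limits below.
# Free-rotation cross constraints tie both side back chains to the middle one at their third segment.
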PhysicsAsset /Game/Characters/Player/Female/Medium/Bodies/F_MED_Alchemy/Meshes/Parts/F_MED_Alchemy_HairBackblingPhysics
bodies {
 clavicle_l [{shapeType=Capsule, length=16.035, radius=3.4268, center=<4.0958 -2.66 4.638>, rotation=<rot 90 0 0>}]
 clavicle_r [{shapeType=Capsule, length=15.1149, radius=4.347, center=<-3.805 2.6608 -4.638>, rotation=<rot 90 0 0>}]
 dyn_frontLockA_01_l [{shapeType=Capsule, length=2.5069, radius=0.866, center=<1.4673 -0.0585 0.0174>, rotation=<rot 0 -89.031 -86.5841>}]
 dyn_frontLockA_01_r [{shapeType=Capsule, length=2.5069, radius=0.866, center=<1.4673 -0.0585 0.0174>, rotation=<rot 0 -89.031 -86.5841>}]
 dyn_frontLockA_02_l [{shapeType=Capsule, length=5.2358, radius=0.9285, center=<2.8196 0.065 0.1245>, rotation=<rot 0 -89.3205 -88.2058>}]
 dyn_frontLockA_02_r [{shapeType=Capsule, length=3.4534, radius=0.9285, center=<2.4286 0.0938 0.0484>, rotation=<rot 0 -89.3205 -88.2058>}]
 dyn_frontLockA_03_l [{shapeType=Capsule, length=5.2358, radius=0.9285, center=<2.822 -0.1298 0.1245>, rotation=<rot 0 -89.3205 -88.2058>}]
 dyn_frontLockA_03_r [{shapeType=Capsule, length=5.2358, radius=0.9285, center=<2.822 -0.1298 0.1245>, rotation=<rot 0 -89.3205 -88.2058>}]
 dyn_frontLockB_01_l [{shapeType=Capsule, length=2.5069, radius=0.866, center=<1.4673 -0.0585 0.0174>, rotation=<rot 0 -89.031 -86.5841>}]
 dyn_frontLockB_01_r [{shapeType=Capsule, length=2.5069, radius=0.866, center=<1.4673 -0.0585 0.0174>, rotation=<rot 0 -89.031 -86.5841>}]
 dyn_frontLockB_02_l [{shapeType=Capsule, length=5.2358, radius=0.9285, center=<2.8196 0.065 0.1245>, rotation=<rot 0 -89.3205 -88.2058>}]
 dyn_frontLockC_01_l [{shapeType=Capsule, length=6.5716, radius=0.866, center=<3.4379 -0.0252 -0.1003>, rotation=<rot 0 -89.031 -86.5841>}]
 dyn_frontLockC_01_r [{shapeType=Capsule, length=3.9685, radius=0.866, center=<2.1569 -0.0469 -0.0238>, rotation=<rot 0 -89.031 -86.5841>}]
 dyn_frontLockC_02_r [{shapeType=Capsule, length=2.0616, radius=0.9285, center=<0.9162 0.0758 0.0957>, rotation=<rot 0 -89.3205 -88.2058>}]
 dyn_frontLockC_03_r [{shapeType=Capsule, length=5.2358, radius=0.9285, center=<2.822 -0.1298 0.1245>, rotation=<rot 0 -89.3205 -88.2058>}]
 dyn_frontLockMain_01_Base_l [{shapeType=Capsule, length=0.505, radius=0.505, rotation=<rot 0 -90 0>}]
 dyn_frontLockMain_01_Base_r [{shapeType=Capsule, length=0.505, radius=0.505, rotation=<rot 0 -90 0>}]
 dyn_frontLockMain_01_l [{shapeType=Capsule, length=4.5642, radius=3.1839, center=<3.8437 -2.4 -0.0001>, rotation=<rot 90 0 0>}]
 dyn_frontLockMain_01_r [{shapeType=Capsule, length=6.5416, radius=3.1839, center=<4.9614 -2.4 -0.0001>, rotation=<rot 90 0 0>}]
 dyn_frontLockMain_02_l [{shapeType=Capsule, length=5.6144, radius=3.1839, center=<3.9195 -2 0>, rotation=<rot 90 0 0>}]
 dyn_frontLockMain_02_r [{shapeType=Capsule, length=5.4478, radius=3.1839, center=<2.8273 -1.5629 0>, rotation=<rot 90 0 0>}]
 dyn_frontLockMain_03_l [{shapeType=Capsule, length=4.0273, radius=3.1839, center=<3.4503 -2 -0.0002>, rotation=<rot 90 0 0>}]
 dyn_frontLockMain_03_r [{shapeType=Capsule, length=3.9393, radius=3.1839, center=<2.9744 -1.1386 -0.0003>, rotation=<rot 90 0 0>}]
 dyn_frontLockMain_04_l [{shapeType=Capsule, length=4.5642, radius=3.1839, center=<4.1825 -1.9608 -0.0001>, rotation=<rot 90 0 0>}]
 dyn_frontLockMain_04_r [{shapeType=Capsule, length=4.5642, radius=3.1839, center=<3.0109 -2.4 -0.0001>, rotation=<rot 90 0 0>}]
 dyn_frontLockMain_05_l [{shapeType=Capsule, length=2.4468, radius=3.1839, center=<2.3158 -1.843 0>, rotation=<rot 90 0 0>}]
 dyn_frontLockMain_05_r [{shapeType=Capsule, length=4.5642, radius=3.1839, center=<2.7943 -2.4 -0.0001>, rotation=<rot 90 0 0>}]
 dyn_hairBackMain_01 [{shapeType=Capsule, length=4.5642, radius=3.1839, center=<3.8437 -2.4 -0.0001>, rotation=<rot 90 0 0>}]
 dyn_hairBackMain_01_Base [{shapeType=Capsule, length=0.505, radius=0.505, rotation=<rot 90 0 0>}]
 dyn_hairBackMain_02 [{shapeType=Capsule, length=9.0594, radius=3.1839, center=<5.0902 -2.4 0>, rotation=<rot 90 0 0>}]
 dyn_hairBackMain_03 [{shapeType=Capsule, length=8.4444, radius=4.5, center=<5.428 -2.4 -0.0001>, rotation=<rot 90 0 0>}]
 dyn_hairBackMain_04 [{shapeType=Capsule, length=3.3677, radius=4.3804, center=<3.5878 -2.4 -0.0001>, rotation=<rot 90 0 0>}]
 dyn_hairBackMain_05 [{shapeType=Capsule, length=4.5642, radius=3.1839, center=<2.7943 -2.4 -0.0001>, rotation=<rot 90 0 0>}]
 dyn_hairBackMain_06 [{shapeType=Capsule, length=3.3282, radius=3.1839, center=<2.6923 -2.4 0>, rotation=<rot 90 0 0>}]
 dyn_hairBackSide_01_Base_l [{shapeType=Capsule, length=0.505, radius=0.505, rotation=<rot 90 0 0>}]
 dyn_hairBackSide_01_Base_r [{shapeType=Capsule, length=0.505, radius=0.505, rotation=<rot 90 0 0>}]
 dyn_hairBackSide_01_l [{shapeType=Capsule, length=4.5642, radius=3.1839, center=<3.8437 -2.4 -0.0001>, rotation=<rot 90 0 0>}]
 dyn_hairBackSide_01_r [{shapeType=Capsule, length=4.5642, radius=3.1839, center=<3.8437 -2.4 -0.0001>, rotation=<rot 90 0 0>}]
 dyn_hairBackSide_02_l [{shapeType=Capsule, length=9.0594, radius=3.1839, center=<5.0902 -2.4 0>, rotation=<rot 90 0 0>}]
 dyn_hairBackSide_02_r [{shapeType=Capsule, length=9.0594, radius=3.1839, center=<5.0902 -2.4 0>, rotation=<rot 90 0 0>}]
 dyn_hairBackSide_03_l [{shapeType=Capsule, length=9.7605, radius=3.1839, center=<5.428 -2.4 -0.0001>, rotation=<rot 90 0 0>}]
 dyn_hairBackSide_03_r [{shapeType=Capsule, length=9.7605, radius=3.1839, center=<5.428 -2.4 -0.0001>, rotation=<rot 90 0 0>}]
 dyn_hairBackSide_04_l [{shapeType=Capsule, length=4.5642, radius=3.1839, center=<3.5878 -2.4 -0.0001>, rotation=<rot 90 0 0>}]
 dyn_hairBackSide_04_r [{shapeType=Capsule, length=4.5642, radius=3.1839, center=<3.5878 -2.4 -0.0001>, rotation=<rot 90 0 0>}]
 dyn_hairBackSide_05_l [{shapeType=Capsule, length=4.5642, radius=3.1839, center=<2.7943 -2.4 -0.0001>, rotation=<rot 90 0 0>}]
 dyn_hairBackSide_05_r [{shapeType=Capsule, length=4.5642, radius=3.1839, center=<2.7943 -2.4 -0.0001>, rotation=<rot 90 0 0>}]
 head [{shapeType=Capsule, length=11.3216, radius=4.2319, center=<-0.7144 -3.5069 0.0001>, rotation=<rot 90 -172.875 163.8984>}]
 neck_01 [{shapeType=Capsule, length=6.0073, radius=3.4243, rotation=<rot 90 0 0>}]
 neck_02 [{shapeType=Capsule, length=6.0073, radius=3.4243, rotation=<rot 90 0 0>}]
 spine_03 [
  {shapeType=Capsule, length=9.1757, radius=11.9272, center=<9.9203 0.889 0>, rotation=<rot 0 -90 0>}
  {shapeType=Capsule, length=19.5683, radius=7.0178, center=<9.6764 5.105 0.0001>}
 ]
 spine_04 [{shapeType=Capsule, length=15, radius=10.3882, center=<7.1059 0 0>, rotation=<rot 0 -90 0>}]
 spine_05 [
  {shapeType=Sphere, radius=6.8796, center=<9.8763 0 0>}
  {shapeType=Capsule, length=13.9637, radius=8.6301, center=<4.8865 -2.2352 0>, rotation=<rot 0 -90 0>}
 ]
}
constraints [
 {
  bone1Name=head
  bone2Name=neck_02
  swing1LimitDegrees=45
  swing1Motion=Free
  swing2LimitDegrees=45
  swing2Motion=Free
  twistLimitDegrees=45
  twistMotion=Free
 }
 {
  bone1Name=spine_04
  bone2Name=spine_03
  swing1LimitDegrees=45
  swing1Motion=Free
  swing2LimitDegrees=45
  swing2Motion=Free
  twistLimitDegrees=45
  twistMotion=Free
 }
 {
  bone1Name=spine_05
  bone2Name=spine_04
  swing1LimitDegrees=45
  swing1Motion=Free
  swing2LimitDegrees=45
  swing2Motion=Free
  twistLimitDegrees=45
  twistMotion=Free
 }
 {
  bone1Name=clavicle_l
  bone2Name=spine_05
  swing1LimitDegrees=45
  swing1Motion=Free
  swing2LimitDegrees=45
  swing2Motion=Free
  twistLimitDegrees=45
  twistMotion=Free
 }
 {
  bone1Name=clavicle_r
  bone2Name=spine_05
  swing1LimitDegrees=45
  swing1Motion=Free
  swing2LimitDegrees=45
  swing2Motion=Free
  twistLimitDegrees=45
  twistMotion=Free
 }
 {
  bone1Name=dyn_hairBackMain_01_Base
  bone2Name=head
  swing1LimitDegrees=45
  swing1Motion=Free
  swing2LimitDegrees=45
  swing2Motion=Free
  twistLimitDegrees=45
  twistMotion=Free
 }
 {bone1Name=dyn_hairBackMain_01, bone2Name=dyn_hairBackMain_01_Base, swing1LimitDegrees=16, swing2LimitDegrees=8, twistLimitDegrees=4}
 {bone1Name=dyn_hairBackMain_02, bone2Name=dyn_hairBackMain_01, swing1LimitDegrees=16, swing2LimitDegrees=8, twistLimitDegrees=4}
 {bone1Name=dyn_hairBackMain_03, bone2Name=dyn_hairBackMain_02, swing1LimitDegrees=16, swing2LimitDegrees=8, twistLimitDegrees=4}
 {bone1Name=dyn_hairBackMain_04, bone2Name=dyn_hairBackMain_03, swing1LimitDegrees=16, swing2LimitDegrees=8, twistLimitDegrees=4}
 {bone1Name=dyn_hairBackMain_05, bone2Name=dyn_hairBackMain_04, swing1LimitDegrees=16, swing2LimitDegrees=8, twistLimitDegrees=4}
 {bone1Name=dyn_hairBackMain_06, bone2Name=dyn_hairBackMain_05, swing1LimitDegrees=16, swing2LimitDegrees=8, twistLimitDegrees=4}
 {
  bone1Name=dyn_hairBackSide_01_Base_r
  bone2Name=head
  swing1LimitDegrees=45
  swing1Motion=Free
  swing2LimitDegrees=45
  swing2Motion=Free
  twistLimitDegrees=45
  twistMotion=Free
 }
 {bone1Name=dyn_hairBackSide_01_r, bone2Name=dyn_hairBackSide_01_Base_r, swing1LimitDegrees=16, swing2LimitDegrees=8, twistLimitDegrees=4}
 {bone1Name=dyn_hairBackSide_02_r, bone2Name=dyn_hairBackSide_01_r, swing1LimitDegrees=16, swing2LimitDegrees=8, twistLimitDegrees=4}
 {bone1Name=dyn_hairBackSide_03_r, bone2Name=dyn_hairBackSide_02_r, swing1LimitDegrees=16, swing2LimitDegrees=8, twistLimitDegrees=4}
 {bone1Name=dyn_hairBackSide_04_r, bone2Name=dyn_hairBackSide_03_r, swing1LimitDegrees=16, swing2LimitDegrees=8, twistLimitDegrees=4}
 {bone1Name=dyn_hairBackSide_05_r, bone2Name=dyn_hairBackSide_04_r, swing1LimitDegrees=16, swing2LimitDegrees=8, twistLimitDegrees=4}
 {
  bone1Name=dyn_hairBackSide_01_Base_l
  bone2Name=head
  swing1LimitDegrees=45
  swing1Motion=Free
  swing2LimitDegrees=45
  swing2Motion=Free
  twistLimitDegrees=45
  twistMotion=Free
 }
 {bone1Name=dyn_hairBackSide_01_l, bone2Name=dyn_hairBackSide_01_Base_l, swing1LimitDegrees=16, swing2LimitDegrees=8, twistLimitDegrees=4}
 {bone1Name=dyn_hairBackSide_02_l, bone2Name=dyn_hairBackSide_01_l, swing1LimitDegrees=16, swing2LimitDegrees=8, twistLimitDegrees=4}
 {bone1Name=dyn_hairBackSide_03_l, bone2Name=dyn_hairBackSide_02_l, swing1LimitDegrees=16, swing2LimitDegrees=8, twistLimitDegrees=4}
 {bone1Name=dyn_hairBackSide_04_l, bone2Name=dyn_hairBackSide_03_l, swing1LimitDegrees=16, swing2LimitDegrees=8, twistLimitDegrees=4}
 {bone1Name=dyn_hairBackSide_05_l, bone2Name=dyn_hairBackSide_04_l, swing1LimitDegrees=16, swing2LimitDegrees=8, twistLimitDegrees=4}
 {
  bone1Name=dyn_hairBackSide_03_l
  bone2Name=dyn_hairBackMain_03
  swing1LimitDegrees=45
  swing1Motion=Free
  swing2LimitDegrees=45
  swing2Motion=Free
  twistLimitDegrees=45
  twistMotion=Free
 }
 {
  bone1Name=dyn_hairBackSide_03_r
  bone2Name=dyn_hairBackMain_03
  swing1LimitDegrees=45
  swing1Motion=Free
  swing2LimitDegrees=45
  swing2Motion=Free
  twistLimitDegrees=45
  twistMotion=Free
 }
 {
  bone1Name=dyn_frontLockMain_01_Base_r
  bone2Name=head
  swing1LimitDegrees=45
  swing1Motion=Free
  swing2LimitDegrees=45
  swing2Motion=Free
  twistLimitDegrees=45
  twistMotion=Free
 }
 {bone1Name=dyn_frontLockMain_01_r, bone2Name=dyn_frontLockMain_01_Base_r, swing1LimitDegrees=4, swing2LimitDegrees=4, twistLimitDegrees=4}
 {bone1Name=dyn_frontLockMain_02_r, bone2Name=dyn_frontLockMain_01_r, swing1LimitDegrees=12, swing2LimitDegrees=8, twistLimitDegrees=6}
 {bone1Name=dyn_frontLockMain_03_r, bone2Name=dyn_frontLockMain_02_r, swing1LimitDegrees=12, swing2LimitDegrees=8, twistLimitDegrees=6}
 {bone1Name=dyn_frontLockMain_04_r, bone2Name=dyn_frontLockMain_03_r, swing1LimitDegrees=12, swing2LimitDegrees=8, twistLimitDegrees=6}
 {bone1Name=dyn_frontLockMain_05_r, bone2Name=dyn_frontLockMain_04_r, swing1LimitDegrees=12, swing2LimitDegrees=8, twistLimitDegrees=6}
 {bone1Name=dyn_frontLockB_01_r, bone2Name=dyn_frontLockMain_04_r, swing1LimitDegrees=6, swing2LimitDegrees=6, twistLimitDegrees=4}
 {bone1Name=dyn_frontLockA_01_r, bone2Name=dyn_frontLockMain_03_r, swing1LimitDegrees=10, swing2LimitDegrees=10, twistLimitDegrees=4}
 {bone1Name=dyn_frontLockA_02_r, bone2Name=dyn_frontLockA_01_r, swing1LimitDegrees=10, swing2LimitDegrees=10, twistLimitDegrees=4}
 {bone1Name=dyn_frontLockA_03_r, bone2Name=dyn_frontLockA_02_r, swing1LimitDegrees=10, swing2LimitDegrees=10, twistLimitDegrees=4}
 {bone1Name=dyn_frontLockC_01_r, bone2Name=dyn_frontLockMain_03_r, swing1LimitDegrees=10, swing2LimitDegrees=10, twistLimitDegrees=4}
 {bone1Name=dyn_frontLockC_02_r, bone2Name=dyn_frontLockC_01_r, swing1LimitDegrees=10, swing2LimitDegrees=10, twistLimitDegrees=4}
 {bone1Name=dyn_frontLockC_03_r, bone2Name=dyn_frontLockC_02_r, swing1LimitDegrees=10, swing2LimitDegrees=10, twistLimitDegrees=4}
 {
  bone1Name=dyn_frontLockMain_01_Base_l
  bone2Name=head
  swing1LimitDegrees=45
  swing1Motion=Free
  swing2LimitDegrees=45
  swing2Motion=Free
  twistLimitDegrees=45
  twistMotion=Free
 }
 {bone1Name=dyn_frontLockMain_01_l, bone2Name=dyn_frontLockMain_01_Base_l, swing1LimitDegrees=4, swing2LimitDegrees=4, twistLimitDegrees=4}
 {bone1Name=dyn_frontLockMain_02_l, bone2Name=dyn_frontLockMain_01_l, swing1LimitDegrees=20, swing2LimitDegrees=8, twistLimitDegrees=6}
 {bone1Name=dyn_frontLockMain_03_l, bone2Name=dyn_frontLockMain_02_l, swing1LimitDegrees=20, swing2LimitDegrees=8, twistLimitDegrees=6}
 {bone1Name=dyn_frontLockMain_04_l, bone2Name=dyn_frontLockMain_03_l, swing1LimitDegrees=14, swing2LimitDegrees=8, twistLimitDegrees=6}
 {
  bone1Name=dyn_frontLockMain_05_l
  bone2Name=dyn_frontLockMain_04_l
  swing1LimitDegrees=12
  swing1Motion=Free
  swing2LimitDegrees=8
  swing2Motion=Free
  twistLimitDegrees=6
  twistMotion=Free
 }
 {bone1Name=dyn_frontLockB_01_l, bone2Name=dyn_frontLockMain_04_l, swing1LimitDegrees=10, swing2LimitDegrees=10, twistLimitDegrees=4}
 {bone1Name=dyn_frontLockB_02_l, bone2Name=dyn_frontLockB_01_l, swing1LimitDegrees=10, swing2LimitDegrees=10, twistLimitDegrees=4}
 {bone1Name=dyn_frontLockA_01_l, bone2Name=dyn_frontLockMain_03_l, swing1LimitDegrees=10, swing2LimitDegrees=10, twistLimitDegrees=4}
 {bone1Name=dyn_frontLockA_02_l, bone2Name=dyn_frontLockA_01_l, swing1LimitDegrees=10, swing2LimitDegrees=10, twistLimitDegrees=4}
 {bone1Name=dyn_frontLockA_03_l, bone2Name=dyn_frontLockA_02_l, swing1LimitDegrees=10, swing2LimitDegrees=10, twistLimitDegrees=4}
 {bone1Name=dyn_frontLockC_01_l, bone2Name=dyn_frontLockMain_03_l, swing1LimitDegrees=10, swing2LimitDegrees=10, twistLimitDegrees=4}
 {
  bone1Name=neck_02
  bone2Name=neck_01
  swing1LimitDegrees=45
  swing1Motion=Free
  swing2LimitDegrees=45
  swing2Motion=Free
  twistLimitDegrees=45
  twistMotion=Free
 }
 {
  bone1Name=neck_01
  bone2Name=spine_05
  swing1LimitDegrees=45
  swing1Motion=Free
  swing2LimitDegrees=45
  swing2Motion=Free
  twistLimitDegrees=45
  twistMotion=Free
 }
]
