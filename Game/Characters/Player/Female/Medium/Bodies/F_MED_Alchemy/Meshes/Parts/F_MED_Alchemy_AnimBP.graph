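# Animation Blueprint for the F_MED_Alchemy female player body part, driving hair dynamics and face-accessory rigs on a pose copied from the attached parent mesh.
# On initialization and customization load, a backbling check selects one of two RigidBody hair simulations, treating capes as no backpack.
# Rotation constraints then tie braids to the main back hair.
# Front-end-only tweaks adjust the right front lock before a post Control Rig outputs the pose.
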AnimBlueprint /Game/Characters/Player/Female/Medium/Bodies/F_MED_Alchemy/Meshes/Parts/F_MED_Alchemy_AnimBP
target_skeleton=&/F_MED_Alchemy_Skeleton
graphs [
 AnimGraph {nodes [
   n1 = "Constraint - dyn_hairBackBraid_02_r (R)"(Alpha=1, ConstraintWeights_0=1) {
    alphaBoolBlend {blendOption=Linear}
    boneToModify {boneName=dyn_hairBackBraid_02_r}
    constraintSetup [{transformType=Rotation, targetBone {boneName=dyn_hairBackMain_03}}]
    constraintWeights [1]
   }
   n2 = "Constraint - dyn_hairBackBraid_03_l (R)"(ComponentPose=@n1, Alpha=1, ConstraintWeights_0=1) {
    alphaBoolBlend {blendOption=Linear}
    boneToModify {boneName=dyn_hairBackBraid_03_l}
    constraintSetup [{transformType=Rotation, targetBone {boneName=dyn_hairBackMain_04}}]
    constraintWeights [1]
   }
   n3 = "Constraint - dyn_frontLockMain_02_r (R)"(ComponentPose=@n2, Alpha=1, ConstraintWeights_0=1) {
    alphaBoolBlend {blendOption=Linear}
    boneToModify {boneName=dyn_frontLockMain_02_r}
    constraintSetup [{transformType=Rotation, targetBone {boneName=dyn_hairBackMain_03}}]
    constraintWeights [1]
   }
   "Constraint - dyn_frontLockMain_03_r (R)"(ComponentPose=@n3, Alpha=1, ConstraintWeights_0=1) {
    alphaBoolBlend {blendOption=Linear}
    boneToModify {boneName=dyn_frontLockMain_03_r}
    constraintSetup [{transformType=Rotation, targetBone {boneName=dyn_hairBackMain_04}}]
    constraintWeights [1]
   }
   n4 = Control Rig(Source=Copy Pose From Mesh() {bUseAttachedParent=true}) {
    alphaBoolBlend {blendOption=Linear}
    controlRigAssetReference {blueprintRigClass=&/F_MED_Alchemy_FaceAcc_CtrlRig.F_MED_Alchemy_FaceAcc_CtrlRig_C}
   }
   n5 = RigidBody!(ComponentPose=Local To Component(LocalPose=@n4), bAlphaBoolEnabled:bool=Get wearingBackpack() -> (wearingBackpack:bool), ExternalForce) -> (Pose) {
    alphaInputType=Bool
    bClampLinearTranslationLimitToRefPose=true
    bOverrideWorldGravity=true
    overridePhysicsAsset=&/F_MED_Alchemy_HairBackblingPhysics
    simulationSpace=BaseBoneSpace
    alphaBoolBlend {blendOption=Linear}
    baseBoneRef {boneName=spine_05}
    componentAppliedLinearAccClamp=<2500 2500 4000>
    componentLinearAccScale=<0.4 0.4 0.6>
    componentLinearVelScale=<0.15 0.15 0.15>
    overrideWorldGravity {z=-980}
   }
   n6 = RigidBody(ComponentPose=@n5, bAlphaBoolEnabled=Get noBackpack() -> (noBackpack:bool)) {
    alphaInputType=Bool
    bClampLinearTranslationLimitToRefPose=true
    bOverrideWorldGravity=true
    overridePhysicsAsset=&/F_MED_Alchemy_HairPerfPhysics
    simulationSpace=BaseBoneSpace
    alphaBoolBlend {blendOption=Linear}
    baseBoneRef {boneName=root}
    componentAppliedLinearAccClamp=<3000 3000 3200>
    componentLinearAccScale {z=0.475}
    componentLinearVelScale=<0.4 0.4 0.125>
    overrideWorldGravity {z=-980}
   }
   n7 = "Constraint - dyn_hairBackBraid_02_l (R)"(ComponentPose=@n6, Alpha=1, ConstraintWeights_0=1) {
    alphaBoolBlend {blendOption=Linear}
    boneToModify {boneName=dyn_hairBackBraid_02_l}
    constraintSetup [{transformType=Rotation, targetBone {boneName=dyn_hairBackMain_03}}]
    constraintWeights [1]
   }
   n8 = "Constraint - dyn_hairBackBraid_03_l (R)"(ComponentPose=@n7, Alpha=1, ConstraintWeights_0=1) {
    alphaBoolBlend {blendOption=Linear}
    boneToModify {boneName=dyn_hairBackBraid_03_l}
    constraintSetup [{transformType=Rotation, targetBone {boneName=dyn_hairBackMain_04}}]
    constraintWeights [1]
   }
   n9 = "Constraint - dyn_hairBackBraid_02_r (R)"(ComponentPose=@n8, Alpha=1, ConstraintWeights_0=1) {
    alphaBoolBlend {blendOption=Linear}
    boneToModify {boneName=dyn_hairBackBraid_02_r}
    constraintSetup [{transformType=Rotation, targetBone {boneName=dyn_hairBackMain_03}}]
    constraintWeights [1]
   }
   n10 = "Constraint - dyn_hairBackBraid_03_r (R)"(ComponentPose=@n9, Alpha=1, ConstraintWeights_0=1) {
    alphaBoolBlend {blendOption=Linear}
    boneToModify {boneName=dyn_hairBackBraid_03_r}
    constraintSetup [{transformType=Rotation, targetBone {boneName=dyn_hairBackMain_04}}]
    constraintWeights [1]
   }
   n11 = Get bIsInFrontEnd()
   n12 = "Transform (Modify) Bone - Bone: dyn_frontLockMain_03_r"!(ComponentPose=@n10, bAlphaBoolEnabled:bool=@n11, Translation="0,0,0", Rotation="0,-8,0", Scale="1,1,1") -> (Pose) {
    alphaInputType=Bool
    rotationMode=BMM_Additive
    rotationSpace=BCS_ParentBoneSpace
    alphaBoolBlend {blendOption=Linear}
    boneToModify {boneName=dyn_frontLockMain_03_r}
    rotation {yaw=-8}
   }
   n13 = "Constraint - dyn_frontLockC_01_r (P)"!(ComponentPose=@n12, bAlphaBoolEnabled:bool=@n11, ConstraintWeights_0=1) -> (Pose) {
    alphaInputType=Bool
    alphaBoolBlend {blendOption=Linear}
    boneToModify {boneName=dyn_frontLockC_01_r}
    constraintSetup [{targetBone {boneName=dyn_frontLockMain_03_r}}]
    constraintWeights [1]
   }
   n14 = "Constraint - dyn_frontLockC_02_r (P)"!(ComponentPose=@n13, bAlphaBoolEnabled:bool=@n11, ConstraintWeights_0=1) -> (Pose) {
    alphaInputType=Bool
    alphaBoolBlend {blendOption=Linear}
    boneToModify {boneName=dyn_frontLockC_02_r}
    constraintSetup [{targetBone {boneName=dyn_frontLockMain_03_r}}]
    constraintWeights [1]
   }
   n15 = "Constraint - dyn_frontLockC_03_r (P)"!(ComponentPose=@n14, bAlphaBoolEnabled:bool=@n11, ConstraintWeights_0=1) -> (Pose) {
    alphaInputType=Bool
    alphaBoolBlend {blendOption=Linear}
    boneToModify {boneName=dyn_frontLockC_03_r}
    constraintSetup [{targetBone {boneName=dyn_frontLockMain_03_r}}]
    constraintWeights [1]
   }
   n16 = Control Rig!(Source=Component To Local(ComponentPose=@n15), bAlphaBoolEnabled:bool=True) -> (Pose) {
    alphaInputType=Bool
    alphaBoolBlend {blendOption=Linear}
    controlRigAssetReference {blueprintRigClass=&/F_MED_Alchemy_FaceAcc_Post_CtrlRig.F_MED_Alchemy_FaceAcc_Post_CtrlRig_C}
   }
   Output Pose(Result=@n16) {layerGroup=None}
  ]}
 EventGraph {nodes [
   n1 = Event OnCharacterCustomizationLoaded()
   n2 = CheckForBackbling()
   n3 = Set wearingBackpack(execute=[@n1.then, @n4.then], wearingBackpack:bool=AND Boolean(A=@n2, B=NOT Boolean(A=@n2.bIsBackblingACape))) -> (Output_Get:bool)
   > Set noBackpack(noBackpack:bool=NOT Boolean(A=@n3)) -> (Output_Get:bool)
   n4 = Event BlueprintInitializeAnimation()
   GetOwningActor()
   Event BlueprintUpdateAnimation()
  ]}
]
variables [
 noBackpack {type=boolean}
 wearingBackpack {type=boolean}
]
